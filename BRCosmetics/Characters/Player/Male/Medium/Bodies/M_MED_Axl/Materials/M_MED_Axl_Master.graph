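# Master skeletal-mesh character material for the M_MED_Axl player body, layering switchable cloth fuzz and Slurp skin effects over the shared PreFX base.
# A skin FX mask, routed through a glitch function, supplies channel masks for those effects.
# One chain adds rim, de-rez and res-in, with optional panning-noise ninja dissolve on opacity mask and vertex-color element hiding through World Position Offset.
# A second chain uses shared PostFX with unconditional vertex-color hiding.
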Material /BRCosmetics/Characters/Player/Male/Medium/Bodies/M_MED_Axl/Materials/M_MED_Axl_Master
graph {
 properties {
  bUseMaterialAttributes=true
  bUsedWithMeshDeformer=true
  bUsedWithMorphTargets=true
  bUsedWithSkeletalMesh=true
  pixelDepthOffsetMode=PDOM_Legacy
  usedShadingModels=Default Lit
 }
 nodes [
  n1 = MaterialFunctionCall() {materialFunction=&BaseFN_CharacterFunction_PreFX}
  n2 = MaterialFunctionCall!(In=TextureObjectParameter() {group=Skin FX, parameterName=SkinFX_Mask}) {materialFunction=/Game/Weapons/WeaponSkins/MaterialFunction/MF_GlitchSubfunction}
  n3 = StaticSwitchParameter(True=@n2.Glitched, False=@n2.Glitched) {group=Skin FX, parameterName=UseGlitch}
  n4 = StaticSwitchParameter(
   True=MaterialFunctionCall!(Alpha=StaticComponentMaskParameter(None=@n3) {defaultR=true, group=Skin FX, parameterName=Cloth Channel}, Gmap=@n1.Gmap, Material=@n1.Gmap) {materialFunction=/Game/Weapons/WeaponSkins/MaterialFunction/ClothFuzz}.Result
   False=@n1.Result
  ) {group=Skin FX, parameterName=Use Cloth fuzz}
  n5 = StaticSwitchParameter(
   True=MaterialFunctionCall!(
    Alpha=StaticComponentMaskParameter(None=@n3) {defaultR=true, group=Skin FX, parameterName=Slurp Channel}
    AO=@n1.AO
    BubblesMask=StaticComponentMaskParameter(None=@n3) {defaultG=true, group=Skin FX, parameterName=Slurp Bubbles Channel}
    SlurpNormalDistortionMask=StaticComponentMaskParameter(None=@n3) {defaultB=true, group=Skin FX, parameterName=Slurp Normal Distortion Channel}
    Pre Skinned Position
    Material=@n4
   ) {materialFunction=/Game/Weapons/WeaponSkins/MaterialFunction/MF_Slurp}.Result
   False=@n4
  ) {defaultValue=true, group=Skin FX, parameterName=Use Slurp}
  n6 = GetMaterialAttributes(None=MaterialFunctionCall!(
    Fresnel
    Material=MaterialFunctionCall!(AO=@n1.AO, Distance Blend=@n1.AO, Gmap, Pre Skinned Local Position, Material=@n5) {materialFunction=/BRCosmetics/Packages/Fortress_SharedMaterials/Base_Material_Functions/MF_Rimv2}.Result
   ) {materialFunction=/Game/Packages/Fortress_SharedMaterials/Base_Material_Functions/CharacterDeRez_Transition}.Result) {attributeGetTypes [#1, #2]}
  n7 = MaterialFunctionCall!(Animation Gradient, Rez In Percentage) {materialFunction=/Game/Packages/Fortress_SharedMaterials/Base_Material_Functions/CharacterResInEffect}
  n8 = TextureCoordinate() {vTiling=4}
  n9 = TextureCoordinate() {vTiling=8}
  n10 = Add(
   A=TextureSample(UVs=Panner(Coordinate=@n8) {speedX=0.65, speedY=0.5}) {texture=&T_NoisePack_01}.R
   B=TextureSample(UVs=Panner(Coordinate=@n9) {speedX=-1.325, speedY=-0.9}) {texture=&T_NoisePack_01}.B
  )
  n11 = Multiply(A=@n7.Modulative Opacity, B=Add(A=Saturate(None=@n10), B=ScalarParameter() {defaultValue=1, parameterName=NinjaDissolve}) {constB=0})
  SetMaterialAttributes!(
   MaterialAttributes=@n6.MaterialAttributes
   Opacity Mask=Multiply(A=@n6.Opacity Mask, B=StaticSwitchParameter(True=@n11, False=@n7.Modulative Opacity) {parameterName=UseNinjaShadowStance?})
   World Position Offset=StaticSwitchParameter(
    True=MaterialFunctionCall() {materialFunction=&/Slurp_HideElementsWithVertexColor}.Result
    False=@n6.World Position Offset
   ) {group=Opacity Mask, parameterName=Use Vertex Colors for Mask}
  ) {attributeSetTypes [#1, #2]}
  n12 = MaterialFunctionCall(AO=@n1.AO, Distance Blend=@n1.AO, Fresnel=@n1.AO, Material=@n5) {materialFunction=&BaseFN_CharacterFunction_PostFX}
  SetMaterialAttributes!(
   MaterialAttributes=GetMaterialAttributes(None=@n12.Result) {attributeGetTypes [#2]}.MaterialAttributes
   World Position Offset=MaterialFunctionCall() {materialFunction=&Slurp_HideElementsWithVertexColor}.Result
  ) {attributeSetTypes [#2]}
 ]
}
