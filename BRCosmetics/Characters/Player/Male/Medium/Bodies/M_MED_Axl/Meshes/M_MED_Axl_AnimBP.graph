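# Post-process style AnimBP for the M_MED_Axl player body cosmetic, adding secondary motion and corrective shaping to the shirt.
# Copies the parent mesh pose through a Control Rig, applies additive shirt bone rotations, then chains spring-based AnimDynamics, many at zero alpha.
# EventGraph derives gravity, simulation space, alphas and shirt bone correctives from movement state, velocity history, and spine, thigh, pelvis and tail transforms.
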AnimBlueprint /BRCosmetics/Characters/Player/Male/Medium/Bodies/M_MED_Axl/Meshes/M_MED_Axl_AnimBP
target_skeleton=&/M_MED_Axl_Skeleton
graphs [
 AnimGraph {nodes [
   n1 = Anim Dynamics - Bone: dyn_shirt_bk_r_01!(
    SimulationSpace:AnimPhysSimSpaceType Enum=Component
    ComponentPose=Local To Component(LocalPose=Use cached pose 'CR_Cache_Transforms'())
    Alpha=0
    ExternalForce="0,0,0"
   ) -> (Pose) {
    angularDampingOverride=1
    angularSpringConstant=20000
    bAngularSpring=true
    bOverrideAngularDamping=true
    bUseGravityOverride=true
    gravityScale=0.35
    alphaBoolBlend {blendOption=Linear}
    boundBone {boneName=dyn_shirt_bk_r_01}
    componentAppliedLinearAccClamp=<3000 3000 3000>
    componentLinearAccScale=<0.3 0.3 0.3>
    componentLinearVelScale=<0.3 0.3 0.3>
    gravityOverride {z=-980}
    physicsBodyDefinitions [{
      boundBone {boneName=dyn_shirt_bk_r_01}
      boxExtents=<40 60 60>
      constraintSetup {angularLimitsMax=<1 1 15>, angularLimitsMin=<-15 -15 -15>, angularTarget=<1 0 0>}
      localJointOffset=<1 0 0>
     }]
   }
   n2 = Anim Dynamics - Bone: dyn_shirt_bk_r_02!(SimulationSpace:AnimPhysSimSpaceType Enum=Component, ComponentPose=@n1, Alpha=0, ExternalForce="0,0,0") -> (Pose) {
    angularDampingOverride=1
    angularSpringConstant=20000
    bAngularSpring=true
    bOverrideAngularDamping=true
    bUseGravityOverride=true
    gravityScale=0.35
    alphaBoolBlend {blendOption=Linear}
    boundBone {boneName=dyn_shirt_bk_r_02}
    componentAppliedLinearAccClamp=<3000 3000 3000>
    componentLinearAccScale=<0.3 0.3 0.3>
    componentLinearVelScale=<0.3 0.3 0.3>
    gravityOverride {z=-980}
    physicsBodyDefinitions [{
      boundBone {boneName=dyn_shirt_bk_r_02}
      boxExtents=<40 60 60>
      constraintSetup {angularLimitsMax=<1 1 15>, angularLimitsMin=<-15 -15 -15>, angularTarget=<1 0 0>}
      localJointOffset=<1 0 0>
     }]
   }
   n3 = Anim Dynamics - Bone: dyn_shirt_r_01!(SimulationSpace:AnimPhysSimSpaceType Enum=Component, ComponentPose=@n2, Alpha=1, ExternalForce="0,0,0") -> (Pose) {
    angularDampingOverride=1
    angularSpringConstant=20000
    bAngularSpring=true
    bOverrideAngularDamping=true
    bUseGravityOverride=true
    gravityScale=0.35
    alphaBoolBlend {blendOption=Linear}
    boundBone {boneName=dyn_shirt_r_01}
    componentAppliedLinearAccClamp=<3000 3000 3000>
    componentLinearAccScale=<0.3 0.3 0.3>
    componentLinearVelScale=<0.3 0.3 0.3>
    gravityOverride {z=-980}
    physicsBodyDefinitions [{
      boundBone {boneName=dyn_shirt_r_01}
      boxExtents=<40 60 60>
      constraintSetup {angularLimitsMax=<1 15 15>, angularLimitsMin=<-15 -15 -15>, angularTarget=<1 0 0>}
      localJointOffset=<4 0 0>
     }]
   }
   n4 = Anim Dynamics - Bone: dyn_shirt_r_02!(SimulationSpace:AnimPhysSimSpaceType Enum=Component, ComponentPose=@n3, Alpha=0, ExternalForce="0,0,0") -> (Pose) {
    angularDampingOverride=1
    angularSpringConstant=20000
    bAngularSpring=true
    bOverrideAngularDamping=true
    bUseGravityOverride=true
    gravityScale=0.35
    alphaBoolBlend {blendOption=Linear}
    boundBone {boneName=dyn_shirt_r_02}
    componentAppliedLinearAccClamp=<3000 3000 3000>
    componentLinearAccScale=<0.3 0.3 0.3>
    componentLinearVelScale=<0.3 0.3 0.3>
    gravityOverride {z=-980}
    physicsBodyDefinitions [{
      boundBone {boneName=dyn_shirt_r_02}
      boxExtents=<40 60 60>
      constraintSetup {angularLimitsMax=<1 15 15>, angularLimitsMin=<-15 -15 -15>, angularTarget=<1 0 0>}
      localJointOffset=<4 0 0>
     }]
   }
   n5 = Anim Dynamics - Bone: dyn_shirt_frmid_r_01!(SimulationSpace:AnimPhysSimSpaceType Enum=Component, ComponentPose=@n4, Alpha=1, ExternalForce="0,0,0") -> (Pose) {
    angularDampingOverride=1
    angularSpringConstant=20000
    bAngularSpring=true
    bOverrideAngularDamping=true
    bUseGravityOverride=true
    gravityScale=0.35
    alphaBoolBlend {blendOption=Linear}
    boundBone {boneName=dyn_shirt_frmid_r_01}
    componentAppliedLinearAccClamp=<3000 3000 3000>
    componentLinearAccScale=<0.3 0.3 0.3>
    componentLinearVelScale=<0.3 0.3 0.3>
    gravityOverride {z=-980}
    physicsBodyDefinitions [{
      boundBone {boneName=dyn_shirt_frmid_r_01}
      boxExtents=<40 60 60>
      constraintSetup {angularLimitsMax=<10 1 5>, angularLimitsMin=<-15 -10 -5>, angularTarget=<1 0 0>}
      localJointOffset=<4 0 0>
     }]
   }
   n6 = Anim Dynamics - Bone: dyn_shirt_frmid_r_02!(SimulationSpace:AnimPhysSimSpaceType Enum=Component, ComponentPose=@n5, Alpha=0, ExternalForce="0,0,0") -> (Pose) {
    angularDampingOverride=1
    angularSpringConstant=20000
    bAngularSpring=true
    bOverrideAngularDamping=true
    bUseGravityOverride=true
    gravityScale=0.35
    alphaBoolBlend {blendOption=Linear}
    boundBone {boneName=dyn_shirt_frmid_r_02}
    componentAppliedLinearAccClamp=<3000 3000 3000>
    componentLinearAccScale=<0.3 0.3 0.3>
    componentLinearVelScale=<0.3 0.3 0.3>
    gravityOverride {z=-980}
    physicsBodyDefinitions [{
      boundBone {boneName=dyn_shirt_frmid_r_02}
      boxExtents=<40 60 60>
      constraintSetup {angularLimitsMax=<10 1 5>, angularLimitsMin=<-15 -10 -5>, angularTarget=<1 0 0>}
      localJointOffset=<4 0 0>
     }]
   }
   n7 = Anim Dynamics - Bone: dyn_shirt_bk_01!(SimulationSpace:AnimPhysSimSpaceType Enum=Component, ComponentPose=@n6, Alpha=0, ExternalForce="0,0,0") -> (Pose) {
    angularDampingOverride=1
    angularSpringConstant=20000
    bAngularSpring=true
    bOverrideAngularDamping=true
    bUseGravityOverride=true
    gravityScale=0.35
    alphaBoolBlend {blendOption=Linear}
    boundBone {boneName=dyn_shirt_bk_01}
    componentAppliedLinearAccClamp=<3000 3000 3000>
    componentLinearAccScale=<0.3 0.3 0.3>
    componentLinearVelScale=<0.3 0.3 0.3>
    gravityOverride {z=-980}
    physicsBodyDefinitions [{
      boundBone {boneName=dyn_shirt_bk_01}
      boxExtents=<40 60 60>
      constraintSetup {angularLimitsMax=<1 1 15>, angularLimitsMin=<-15 -15 -15>, angularTarget=<1 0 0>}
      localJointOffset=<1 0 0>
     }]
   }
   n8 = Anim Dynamics - Bone: dyn_shirt_bk_02!(SimulationSpace:AnimPhysSimSpaceType Enum=Component, ComponentPose=@n7, Alpha=0, ExternalForce="0,0,0") -> (Pose) {
    angularDampingOverride=1
    angularSpringConstant=20000
    bAngularSpring=true
    bOverrideAngularDamping=true
    bUseGravityOverride=true
    gravityScale=0.35
    alphaBoolBlend {blendOption=Linear}
    boundBone {boneName=dyn_shirt_bk_02}
    componentAppliedLinearAccClamp=<3000 3000 3000>
    componentLinearAccScale=<0.3 0.3 0.3>
    componentLinearVelScale=<0.3 0.3 0.3>
    gravityOverride {z=-980}
    physicsBodyDefinitions [{
      boundBone {boneName=dyn_shirt_bk_02}
      boxExtents=<40 60 60>
      constraintSetup {angularLimitsMax=<1 1 15>, angularLimitsMin=<-15 -15 -15>, angularTarget=<1 0 0>}
      localJointOffset=<1 0 0>
     }]
   }
   n9 = Anim Dynamics - Bone: dyn_shirt_l_01!(SimulationSpace:AnimPhysSimSpaceType Enum=Component, ComponentPose=@n8, Alpha=0, ExternalForce="0,0,0") -> (Pose) {
    angularDampingOverride=1
    angularSpringConstant=20000
    bAngularSpring=true
    bOverrideAngularDamping=true
    bUseGravityOverride=true
    gravityScale=0.35
    alphaBoolBlend {blendOption=Linear}
    boundBone {boneName=dyn_shirt_l_01}
    componentAppliedLinearAccClamp=<3000 3000 3000>
    componentLinearAccScale=<0.3 0.3 0.3>
    componentLinearVelScale=<0.3 0.3 0.3>
    gravityOverride {z=-980}
    physicsBodyDefinitions [{
      boundBone {boneName=dyn_shirt_l_01}
      boxExtents=<40 60 60>
      constraintSetup {angularLimitsMax=<1 1 1>, angularLimitsMin=<-15 -5 -5>, angularTarget=<1 0 0>}
      localJointOffset=<4 0 0>
     }]
   }
   n10 = Anim Dynamics - Bone: dyn_shirt_l_02!(SimulationSpace:AnimPhysSimSpaceType Enum=Component, ComponentPose=@n9, Alpha=0, ExternalForce="0,0,0") -> (Pose) {
    angularDampingOverride=1
    angularSpringConstant=20000
    bAngularSpring=true
    bOverrideAngularDamping=true
    bUseGravityOverride=true
    gravityScale=0.35
    alphaBoolBlend {blendOption=Linear}
    boundBone {boneName=dyn_shirt_l_02}
    componentAppliedLinearAccClamp=<3000 3000 3000>
    componentLinearAccScale=<0.3 0.3 0.3>
    componentLinearVelScale=<0.3 0.3 0.3>
    gravityOverride {z=-980}
    physicsBodyDefinitions [{
      boundBone {boneName=dyn_shirt_l_02}
      boxExtents=<40 60 60>
      constraintSetup {angularLimitsMax=<1 1 1>, angularLimitsMin=<-15 -5 -5>, angularTarget=<1 0 0>}
      localJointOffset=<4 0 0>
     }]
   }
   n11 = Anim Dynamics - Bone: dyn_shirt_frmid_l_01!(SimulationSpace:AnimPhysSimSpaceType Enum=Component, ComponentPose=@n10, Alpha=1, ExternalForce="0,0,0") -> (Pose) {
    angularDampingOverride=1
    angularSpringConstant=20000
    bAngularSpring=true
    bOverrideAngularDamping=true
    bUseGravityOverride=true
    gravityScale=0.35
    alphaBoolBlend {blendOption=Linear}
    boundBone {boneName=dyn_shirt_frmid_l_01}
    componentAppliedLinearAccClamp=<3000 3000 3000>
    componentLinearAccScale=<0.3 0.3 0.3>
    componentLinearVelScale=<0.3 0.3 0.3>
    gravityOverride {z=-980}
    physicsBodyDefinitions [{
      boundBone {boneName=dyn_shirt_frmid_l_01}
      boxExtents=<40 60 60>
      constraintSetup {angularLimitsMax=<10 1 5>, angularLimitsMin=<-15 -10 -5>, angularTarget=<1 0 0>}
      localJointOffset=<4 0 0>
     }]
   }
   n12 = Anim Dynamics - Bone: dyn_shirt_frmid_l_02!(SimulationSpace:AnimPhysSimSpaceType Enum=Component, ComponentPose=@n11, Alpha=1, ExternalForce="0,0,0") -> (Pose) {
    angularDampingOverride=1
    angularSpringConstant=20000
    bAngularSpring=true
    bOverrideAngularDamping=true
    bUseGravityOverride=true
    gravityScale=0.35
    alphaBoolBlend {blendOption=Linear}
    boundBone {boneName=dyn_shirt_frmid_l_02}
    componentAppliedLinearAccClamp=<3000 3000 3000>
    componentLinearAccScale=<0.3 0.3 0.3>
    componentLinearVelScale=<0.3 0.3 0.3>
    gravityOverride {z=-980}
    physicsBodyDefinitions [{
      boundBone {boneName=dyn_shirt_frmid_l_02}
      boxExtents=<40 60 60>
      constraintSetup {angularLimitsMax=<10 1 5>, angularLimitsMin=<-15 -10 -5>, angularTarget=<1 0 0>}
      localJointOffset=<4 0 0>
     }]
   }
   n13 = Anim Dynamics - Bone: dyn_shirt_fr_01!(SimulationSpace:AnimPhysSimSpaceType Enum=Component, ComponentPose=@n12, Alpha=1, ExternalForce="0,0,0") -> (Pose) {
    angularDampingOverride=1
    angularSpringConstant=20000
    bAngularSpring=true
    bOverrideAngularDamping=true
    bUseGravityOverride=true
    gravityScale=0.35
    alphaBoolBlend {blendOption=Linear}
    boundBone {boneName=dyn_shirt_fr_01}
    componentAppliedLinearAccClamp=<3000 3000 3000>
    componentLinearAccScale=<0.3 0.3 0.3>
    componentLinearVelScale=<0.3 0.3 0.3>
    gravityOverride {z=-980}
    physicsBodyDefinitions [{
      boundBone {boneName=dyn_shirt_fr_01}
      boxExtents=<40 60 60>
      constraintSetup {angularLimitsMax=<10 1 5>, angularLimitsMin=<-5 -10 -5>, angularTarget=<1 0 0>}
      localJointOffset=<4 0 0>
     }]
   }
   n14 = Anim Dynamics - Bone: dyn_shirt_fr_02!(SimulationSpace:AnimPhysSimSpaceType Enum=Component, ComponentPose=@n13, Alpha=1, ExternalForce="0,0,0") -> (Pose) {
    angularDampingOverride=1
    angularSpringConstant=20000
    bAngularSpring=true
    bOverrideAngularDamping=true
    bUseGravityOverride=true
    gravityScale=0.35
    alphaBoolBlend {blendOption=Linear}
    boundBone {boneName=dyn_shirt_fr_02}
    componentAppliedLinearAccClamp=<3000 3000 3000>
    componentLinearAccScale=<0.3 0.3 0.3>
    componentLinearVelScale=<0.3 0.3 0.3>
    gravityOverride {z=-980}
    physicsBodyDefinitions [{
      boundBone {boneName=dyn_shirt_fr_02}
      boxExtents=<40 60 60>
      constraintSetup {angularLimitsMax=<10 1 5>, angularLimitsMin=<-5 -10 -5>, angularTarget=<1 0 0>}
      localJointOffset=<4 0 0>
     }]
   }
   n15 = Anim Dynamics - Bone: dyn_shirt_bk_l_01!(SimulationSpace:AnimPhysSimSpaceType Enum=Component, ComponentPose=@n14, Alpha=0, ExternalForce="0,0,0") -> (Pose) {
    angularDampingOverride=1
    angularSpringConstant=20000
    bAngularSpring=true
    bOverrideAngularDamping=true
    bUseGravityOverride=true
    gravityScale=0.35
    alphaBoolBlend {blendOption=Linear}
    boundBone {boneName=dyn_shirt_bk_l_01}
    componentAppliedLinearAccClamp=<3000 3000 3000>
    componentLinearAccScale=<0.3 0.3 0.3>
    componentLinearVelScale=<0.3 0.3 0.3>
    gravityOverride {z=-980}
    physicsBodyDefinitions [{
      boundBone {boneName=dyn_shirt_bk_l_01}
      boxExtents=<40 60 60>
      constraintSetup {angularLimitsMax=<1 1 15>, angularLimitsMin=<-15 -15 -15>, angularTarget=<1 0 0>}
      localJointOffset=<1 0 0>
     }]
   }
   n16 = Anim Dynamics - Bone: dyn_shirt_bk_l_02!(SimulationSpace:AnimPhysSimSpaceType Enum=Component, ComponentPose=@n15, Alpha=0, ExternalForce="0,0,0") -> (Pose) {
    angularDampingOverride=1
    angularSpringConstant=20000
    bAngularSpring=true
    bOverrideAngularDamping=true
    bUseGravityOverride=true
    gravityScale=0.35
    alphaBoolBlend {blendOption=Linear}
    boundBone {boneName=dyn_shirt_bk_l_02}
    componentAppliedLinearAccClamp=<3000 3000 3000>
    componentLinearAccScale=<0.3 0.3 0.3>
    componentLinearVelScale=<0.3 0.3 0.3>
    gravityOverride {z=-980}
    physicsBodyDefinitions [{
      boundBone {boneName=dyn_shirt_bk_l_02}
      boxExtents=<40 60 60>
      constraintSetup {angularLimitsMax=<1 1 15>, angularLimitsMin=<-15 -15 -15>, angularTarget=<1 0 0>}
      localJointOffset=<1 0 0>
     }]
   }
   Output Pose(Result=Component To Local(ComponentPose=@n16)) {layerGroup=None}
   n17 = Control Rig!(
    Source=Copy Pose From Mesh() {bUseAttachedParent=true}
    Alpha
    Pub_IsTail:double=Get TailAlpha() -> (TailAlpha:double)
    Pub_IsCrouching:double=Get crouch alpha() -> (crouch alpha:double)
    Pub_Ismoving:double=Get isMoving() -> (isMoving:double)
   ) -> (Pose) {
    alphaBoolBlend {blendOption=Linear}
    controlRigAssetReference {blueprintRigClass=&/M_MED_Axl_CtrlRig.M_MED_Axl_CtrlRig_C}
   }
   Save cached pose 'CR_Cache_input'(Pose=@n17) {cacheName=CR_Cache_input}
   n18 = "Transform (Modify) Bone - Bone: base_shirt_bk_l"(
    ComponentPose=Local To Component(LocalPose=Use cached pose 'CR_Cache_input'())
    Alpha=1
    Translation="0,0,0"
    Rotation="30,0,0"
    Scale="1,1,1"
   ) {
    rotationMode=BMM_Additive
    rotationSpace=BCS_BoneSpace
    translationMode=BMM_Additive
    translationSpace=BCS_BoneSpace
    alphaBoolBlend {blendOption=Linear}
    alphaScaleBiasClamp {bInterpResult=true, interpSpeedDecreasing=5, interpSpeedIncreasing=5}
    boneToModify {boneName=base_shirt_bk_l}
    rotation {pitch=30}
   }
   n19 = "Transform (Modify) Bone - Bone: base_shirt_bk"(ComponentPose=@n18, Alpha=1, Translation="0,0,0", Rotation="40,0,10", Scale="1,1,1") {
    rotationMode=BMM_Additive
    rotationSpace=BCS_BoneSpace
    translationMode=BMM_Additive
    translationSpace=BCS_BoneSpace
    alphaBoolBlend {blendOption=Linear}
    alphaScaleBiasClamp {bInterpResult=true, interpSpeedDecreasing=5, interpSpeedIncreasing=5}
    boneToModify {boneName=base_shirt_bk}
    rotation {pitch=40, roll=10}
   }
   n20 = "Transform (Modify) Bone - Bone: base_shirt_bk_r"(ComponentPose=@n19, Alpha=1, Translation="0,0,0", Rotation="30,0,0", Scale="1,1,1") {
    rotationMode=BMM_Additive
    rotationSpace=BCS_BoneSpace
    translationMode=BMM_Additive
    translationSpace=BCS_BoneSpace
    alphaBoolBlend {blendOption=Linear}
    alphaScaleBiasClamp {bInterpResult=true, interpSpeedDecreasing=5, interpSpeedIncreasing=5}
    boneToModify {boneName=base_shirt_bk_r}
    rotation {pitch=30}
   }
   n21 = "Transform (Modify) Bone - Bone: base_shirt_bk_l"(ComponentPose=@n20, Alpha=1, Translation="0,0,0", Rotation="-50,0,0", Scale="1,1,1") {
    rotationMode=BMM_Additive
    rotationSpace=BCS_BoneSpace
    translationMode=BMM_Additive
    translationSpace=BCS_BoneSpace
    alphaBoolBlend {blendOption=Linear}
    alphaScaleBiasClamp {bInterpResult=true, interpSpeedDecreasing=5, interpSpeedIncreasing=5}
    boneToModify {boneName=base_shirt_bk_l}
    rotation {pitch=-50}
   }
   n22 = "Transform (Modify) Bone - Bone: base_shirt_bk"(ComponentPose=@n21, Alpha=1, Translation="0,0,0", Rotation="-50,0,10", Scale="1,1,1") {
    rotationMode=BMM_Additive
    rotationSpace=BCS_BoneSpace
    translationMode=BMM_Additive
    translationSpace=BCS_BoneSpace
    alphaBoolBlend {blendOption=Linear}
    alphaScaleBiasClamp {bInterpResult=true, interpSpeedDecreasing=5, interpSpeedIncreasing=5}
    boneToModify {boneName=base_shirt_bk}
    rotation {pitch=-50, roll=10}
   }
   n23 = "Transform (Modify) Bone - Bone: base_shirt_bk_r"(ComponentPose=@n22, Alpha=1, Translation="0,0,0", Rotation="-50,0,0", Scale="1,1,1") {
    rotationMode=BMM_Additive
    rotationSpace=BCS_BoneSpace
    translationMode=BMM_Additive
    translationSpace=BCS_BoneSpace
    alphaBoolBlend {blendOption=Linear}
    alphaScaleBiasClamp {bInterpResult=true, interpSpeedDecreasing=5, interpSpeedIncreasing=5}
    boneToModify {boneName=base_shirt_bk_r}
    rotation {pitch=-50}
   }
   n24 = "Transform (Modify) Bone - Bone: base_shirt_frmid_r"(ComponentPose=@n23, Alpha=1, Translation="0,0,0", Rotation="15,0,0", Scale="1,1,1") {
    rotationMode=BMM_Additive
    rotationSpace=BCS_BoneSpace
    translationMode=BMM_Additive
    translationSpace=BCS_BoneSpace
    alphaBoolBlend {blendOption=Linear}
    alphaScaleBiasClamp {bInterpResult=true, interpSpeedDecreasing=5, interpSpeedIncreasing=5}
    boneToModify {boneName=base_shirt_frmid_r}
    rotation {pitch=15}
   }
   Save cached pose 'CR_Cache_Transforms'(Pose=Component To Local(ComponentPose=@n24)) {cacheName=CR_Cache_Transforms}
  ]}
 EventGraph {nodes [
   n1 = Event BlueprintUpdateAnimation()
   > Set Delta Time X(Delta Time X:double=@n1.DeltaTimeX) -> (Output_Get:double)
   > Is Valid(InputObject=Get BodyMesh() -> (BodyMesh:Skeletal Mesh Component Object Reference))
   > n2 = Sequence!(execute) -> (then_0, then_1, then_2:Exec, then_3:Exec, then_4:Exec, then_5:Exec, then_6:Exec, then_7:Exec, then_8:Exec, then_9:Exec, then_10:Exec)
   > n3 = Branch(Condition=Get bIsSwimSprinting() -> (bIsSwimSprinting:bool))
   > Set gravity_override(gravity_override="0, 0, 0")
   n4 = Branch(execute=@n3.else, Condition=Get bIsSkyDiving())
   > n5 = Sequence!(execute) -> (then_0, then_1)
   > n6 = Branch(Condition=Get bIsParachuteOpened())
   > Set gravity_override(gravity_override="0, 0,-300")
   n7 = Branch(execute=@n6.else, Condition="float > float"(A=Get PawnVelocityZ(), B=50) {is=title K2Node_CallFunction})
   > Set gravity_override(gravity_override="0, 0,200")
   Set gravity_override(execute=@n7.else, gravity_override="0, 0,-200")
   Set Component Linear Acc Scale(execute=@n5.then_1, Component Linear Acc Scale:Vector="0.1,0.1,0.1") -> (Output_Get:Vector)
   > Set Component Linear Vel Scale(Component Linear Vel Scale:Vector="0.1,0.1,0.1") -> (Output_Get:Vector)
   n8 = Sequence(execute=@n4.else)
   Set Component Linear Acc Scale(execute=@n8.then_1, Component Linear Acc Scale="0.4,0.35,0.35")
   > Set Component Linear Vel Scale(Component Linear Vel Scale="0.4,0.35,0.35")
   Set gravity_override(execute=@n8.then_0, gravity_override="0, 0,-980")
   n9 = Branch(execute=@n2.then_10, Condition=Get bIsInFrontEnd())
   > Set SimlationSpace(SimlationSpace:AnimPhysSimSpaceType Enum=World) -> (Output_Get:AnimPhysSimSpaceType Enum)
   > Set AlphaFE(AlphaFE:double=0.3) -> (Output_Get:double)
   Set SimlationSpace(execute=@n9.else, SimlationSpace=Component)
   > Set AlphaFE(AlphaFE=1)
   n10 = Get PawnVelocity()
   n11 = Calculate Velocity From Position History!(
    execute=@n2.then_2
    DeltaSeconds:float=Get Delta Time X() -> (Delta Time X:double)
    Position:Vector=ProjectVectorOnToPlane!(V:Vector=@n10, PlaneNormal:Vector="0, 0,1") -> (ReturnValue:Vector)
    History:"Position History Structure (by ref)"=Get History_position() -> (History_position:Position History Structure)
    NumberOfSamples:int=8
    VelocityMin:float=0.f
    VelocityMax:float=777
   ) -> (then, ReturnValue:float)
   > n12 = Set VelocityChange(VelocityChange:double=@n11) -> (Output_Get:double)
   n13 = MapRangeClamped(Value=Vector Length(A=@n10), InRangeA=300, InRangeB=700)
   n14 = SelectFloat(A=0.5, B=0, bPickA=Get bIsSkyDiving())
   > Set TailAlphaTransform(TailAlphaTransform:double="Clamp (Float)"(Value=float + float!(A=@n12, B=@n13, C:double=@n14) -> (ReturnValue), Min=0, Max=1)) -> (Output_Get:double)
   n15 = Get Delta Time X()
   > Set Shirt_bk_corecction_30pfs(Shirt_bk_corecction_30pfs:double=float * float!(A=@n12, B=MapRangeClamped(Value=@n15, InRangeA=0.017, InRangeB=0.03), C:double) -> (ReturnValue) {is=title K2Node_PromotableOperator}) -> (Output_Get:double)
   n16 = Branch(
    execute=@n2.then_6
    Condition=OR Boolean!(A=Get bIsJumping(), B=Get bIsFalling(), C:bool=Get OnZipline() -> (OnZipline:bool)) -> (ReturnValue)
   )
   > Set Jumping_alpha(Jumping_alpha:double=1) -> (Output_Get:double)
   Set Jumping_alpha(execute=@n16.else, Jumping_alpha=0)
   n17 = Branch(execute=@n2.then_7, Condition=Get bIsCrouching())
   > Set IsCrouching+moving(IsCrouching+moving:double=1) -> (Output_Get:double)
   Set IsCrouching+moving(execute=@n17.else, IsCrouching+moving=0)
   n18 = Event BlueprintProcessAnimInputEvent()
   n19 = Branch(execute=@n2.then_8, Condition=IsNamed(self=@n18.AnimInputEvent, InEventName=aie.Player.Zipline.Entered))
   > Set OnZipline(OnZipline:bool=1) -> (Output_Get:bool)
   > Set ZiplineAlpha(ZiplineAlpha:double=0) -> (Output_Get:double)
   Branch(
    execute=@n19.else
    Condition=OR Boolean(A=IsNamed(self=@n18.AnimInputEvent, InEventName=aie.Player.Zipline.Exited), B=Get bIsMoving2D(), C=Get bIsJumping())
   )
   > Set OnZipline(OnZipline=false)
   > Set ZiplineAlpha(ZiplineAlpha=1)
   n20 = Get bIsMoving2D()
   Set isMoving(execute=@n2.then_9, isMoving:double=FInterpTo(Current=Get isMoving(), Target="To Float (Boolean)"(InBool=@n20), DeltaTime=0, InterpSpeed=5)) -> (Output_Get:double)
   n21 = Get BodyMesh()
   Set Spine_03_Rotation(
    execute=@n2.then_3
    Spine_03_Rotation:Rotator=BreakTransform(InTransform=GetDeltaTransformFromRefPose(self=@n21, BoneName=spine_01, BaseName=spine_03)).Rotation
   ) -> (Output_Get:Rotator)
   > Branch(Condition=Get characterHasTail() -> (characterHasTail:bool))
   > Sequence()
   n22 = Get bIsSkyDiving()
   > Set ConstraintShirt_to_Tail(ConstraintShirt_to_Tail:double="Clamp (Float)"(Value="To Float (Boolean)"(InBool=@n22), Min=0, Max=1)) -> (Output_Get:double)
   > n23 = Branch(Condition=Get bIsSkyDiving())
   n24 = BreakTransform(InTransform=GetSocketTransform(self=Get BodyMesh(), InSocketName=Pelvis, TransformSpace=RTS_Component))
   > Set PevisPitchRotation(PevisPitchRotation:double=MapRangeClamped(Value=BreakRotator(InRot=@n24.Rotation).Pitch, InRangeA=0, InRangeB=70)) -> (Output_Get:double)
   > Set dyn_shirt_bk_01(dyn_shirt_bk_01:Vector="Lerp (Vector)"!(A:Vector="0.53,1,1", B:Vector="0.55,1,1", Alpha:float=Get PevisPitchRotation() -> (PevisPitchRotation:double)) -> (ReturnValue:Vector)) -> (Output_Get:Vector)
   > Set dyn_shirt_bk_02(dyn_shirt_bk_02:Vector="0.78,1,1") -> (Output_Get:Vector)
   > Set dyn_shirt_bk_03(dyn_shirt_bk_03:Vector="0.9,1,1") -> (Output_Get:Vector)
   > n25 = Set dy_shirt_bk_l_01(dy_shirt_bk_l_01:Vector="Lerp (Vector)"(A="0.6,1,1", B="0.7,1,1", Alpha=Get PevisPitchRotation())) -> (Output_Get:Vector)
   n26 = "Lerp (Vector)"(A="0.76,1,1", B="0.8,1,1", Alpha=Get PevisPitchRotation())
   > n27 = Set dy_shirt_bk_l_02(dy_shirt_bk_l_02:Vector=@n26) -> (Output_Get:Vector)
   > n28 = Set dy_shirt_bk_l_03(dy_shirt_bk_l_03:Vector=@n26) -> (Output_Get:Vector)
   > Set dyn_shirt_bk_r_01(dyn_shirt_bk_r_01:Vector=@n25) -> (Output_Get:Vector)
   > Set dyn_shirt_bk_r_02(dyn_shirt_bk_r_02:Vector=@n27) -> (Output_Get:Vector)
   > Set dyn_shirt_bk_r_03(dyn_shirt_bk_r_03:Vector=@n28) -> (Output_Get:Vector)
   n29 = "Lerp (Vector)"(A="0.81,1,1", B="0.95,1,1", Alpha=Get PevisPitchRotation())
   > Set dyn_shirt_l_01(dyn_shirt_l_01:Vector=@n29) -> (Output_Get:Vector)
   > Set dyn_shirt_r_01(dyn_shirt_r_01:Vector=@n29) -> (Output_Get:Vector)
   n30 = Branch(execute=@n23.else, Condition=Get bIsDBNO())
   > Set dyn_shirt_bk_01(dyn_shirt_bk_01="0.69,1,1")
   > Set dyn_shirt_bk_02(dyn_shirt_bk_02="0.88,1,1")
   > Set dyn_shirt_bk_03(dyn_shirt_bk_03="0.95,1,1")
   > n31 = Set dy_shirt_bk_l_01(dy_shirt_bk_l_01="0.82,1,1")
   > n32 = Set dy_shirt_bk_l_02(dy_shirt_bk_l_02="0.95,1,1")
   > n33 = Set dy_shirt_bk_l_03(dy_shirt_bk_l_03="0.95,1,1")
   > Set dyn_shirt_bk_r_01(dyn_shirt_bk_r_01=@n31)
   > Set dyn_shirt_bk_r_02(dyn_shirt_bk_r_02=@n32)
   > Set dyn_shirt_bk_r_03(dyn_shirt_bk_r_03=@n33)
   > n34 = Set dyn_shirt_l_01(dyn_shirt_l_01="1,1,1")
   > Set dyn_shirt_r_01(dyn_shirt_r_01=@n34)
   n35 = Branch(execute=@n30.else, Condition=Get bIsInFrontEnd())
   > Set dyn_shirt_bk_01(dyn_shirt_bk_01="0.7,1,1")
   > Set dyn_shirt_bk_02(dyn_shirt_bk_02="0.74,1,1")
   > Set dyn_shirt_bk_03(dyn_shirt_bk_03="1,1,1")
   n36 = BreakTransform(InTransform=GetDeltaTransformFromRefPose(self=Get TailMesh() -> (TailMesh:Skeletal Mesh Component Object Reference), BoneName=C_Tail_A_Base_Jnt, BaseName=C_Tail_A_1_Jnt))
   n37 = BreakRotator(InRot=Get Spine_03_Rotation() -> (Spine_03_Rotation:Rotator))
   n38 = float * float!(
    A=MapRangeClamped(Value=BreakRotator(InRot=@n36.Rotation), InRangeA=-15, InRangeB=15, OutRangeA=0.12, OutRangeB=-0.12)
    B=MapRangeClamped(Value="Absolute (Float)"(A=@n37), InRangeA=0, InRangeB=26, OutRangeA=1, OutRangeB=2)
   ) -> (ReturnValue)
   > Set dy_shirt_bk_l_01(dy_shirt_bk_l_01=MakeVector(X=float + float!(A=0.7, B=@n38) -> (ReturnValue), Y=1, Z=1))
   > n39 = Set dy_shirt_bk_l_02(dy_shirt_bk_l_02="0.75,1,1")
   > n40 = Set dy_shirt_bk_l_03(dy_shirt_bk_l_03="1,1,1")
   n41 = float * float(A=@n38, B=-1)
   > Set dyn_shirt_bk_r_01(dyn_shirt_bk_r_01=MakeVector(X=float + float(A=0.7, B=@n41), Y=1, Z=1))
   > Set dyn_shirt_bk_r_02(dyn_shirt_bk_r_02=@n39)
   > Set dyn_shirt_bk_r_03(dyn_shirt_bk_r_03=@n40)
   n42 = GetSocketTransform(self=Get BodyMesh(), InSocketName=Pelvis, TransformSpace=RTS_Component)
   Set PelvisHeight(execute=@n35.else, PelvisHeight:double=BreakVector(InVec=BreakTransform(InTransform=@n42)).Z) -> (Output_Get:double)
   n43 = Get Spine_03_Rotation()
   > n44 = Calculate Velocity From Position History(
    DeltaSeconds=Get Delta Time X()
    Position=MakeVector(X=BreakRotator(InRot=@n43), Y=0)
    History=Get History_rotation() -> (History_rotation:Position History Structure)
    NumberOfSamples=8
    VelocityMin=10
    VelocityMax=100
   )
   n45 = Vector Length(A=Get PawnVelocity())
   > Set FastRotation(FastRotation:double=float * float(A=Lerp(A=0, B=-0.12, Alpha=@n44), B=MapRangeClamped(Value=@n45, InRangeA=30, InRangeB=500, OutRangeA=1, OutRangeB=0))) -> (Output_Get:double)
   n46 = Get PawnVelocity()
   n47 = Normalize!(A:Vector=Get PawnVelocity(), Tolerance:float=0.01) -> (ReturnValue:Vector) {is=title K2Node_CallFunction}
   n48 = GetActorForwardVector!(self:Actor Object Reference=TryGetPawnOwner()) -> (ReturnValue:Vector)
   > Set ForwardVelocity(ForwardVelocity:double=float * float(A=Vector Length(A=@n46), B=Dot Product!(A:Vector=@n48, B:Vector=@n47) -> (ReturnValue:double))) -> (Output_Get:double)
   n49 = Get PawnVelocity()
   n50 = GetActorRightVector!(self:Actor Object Reference=TryGetPawnOwner()) -> (ReturnValue:Vector)
   n51 = Normalize(A=Get PawnVelocity(), Tolerance=0.01)
   > Set RightVelcoity(RightVelcoity:double=float * float(A=Vector Length(A=@n49), B=Dot Product(A=@n50, B=@n51))) -> (Output_Get:double)
   n52 = Get ForwardVelocity() -> (ForwardVelocity:double)
   n53 = Get Spine_03_Rotation()
   n54 = float + float!(
    A=MapRangeClamped(Value="Absolute (Float)"(A=@n52), InRangeA=10, InRangeB=215, OutRangeB=0.03)
    B=MapRangeClamped(Value=BreakRotator(InRot=@n53).Yaw, InRangeA=-6, InRangeB=4, OutRangeB=-0.15)
    C:double=MapRangeClamped(Value=Get ForwardVelocity(), InRangeA=-50, InRangeB=-215, OutRangeB=0.07)
   ) -> (ReturnValue)
   n55 = BreakRotator(InRot=Get Spine_03_Rotation())
   n56 = "To Float (Boolean)"(InBool=Get bIsOnGround() -> (bIsOnGround:bool))
   n57 = Get PawnVelocity()
   n58 = float + float!(
    A=MapRangeClamped(Value=Vector Length(A=@n57), InRangeA=555, InRangeB=770, OutRangeB=-0.03)
    B=MapRangeClamped(Value=Get ForwardVelocity(), InRangeA=-50, InRangeB=-215, OutRangeB=-0.11)
   ) -> (ReturnValue)
   > Set!(
    execute
    dyn_shirt_bk_01_X:double=float + float!(
     A=0.71
     B=float * float(A=@n54)
     C:double=float * float(A=@n58, B=@n56)
     D:double=MapRangeClamped(Value=@n55.Yaw, InRangeA=-10, InRangeB=-26, OutRangeB=0.1)
    ) -> (ReturnValue)
    dyn_shirt_bk_01_Y:double=1
    dyn_shirt_bk_01_Z:double=1
   ) -> (then, Output_Get:Vector)
   > Set!(execute, dyn_shirt_bk_02_X:double=0.85, dyn_shirt_bk_02_Y:double=1, dyn_shirt_bk_02_Z:double=1) -> (then, Output_Get)
   > Set dyn_shirt_bk_03(dyn_shirt_bk_03="0.95,1,1")
   n59 = MapRangeClamped(Value=BreakRotator(InRot=Get Spine_03_Rotation()), InRangeA=-26, InRangeB=26, OutRangeB=-0.06)
   n60 = Get PawnVelocity()
   n61 = TryGetPawnOwner()
   n62 = float * float(A=Dot Product(A=GetActorForwardVector(self=@n61), B=Normalize(A=@n60, Tolerance=0.01)))
   n63 = MapRangeClamped(Value="Absolute (Float)"(A=float + float!(A=@n62, B=0.7) -> (ReturnValue)), InRangeA=0, InRangeB=0.21, OutRangeA=1, OutRangeB=0)
   n64 = FInterpEaseInOut!(A:double=0, B:double=-0.1, Alpha:double=@n63, Exponent:double=2) -> (ReturnValue:double)
   n65 = MapRangeClamped(Value=BreakRotator(InRot=Get Spine_03_Rotation()), InRangeA=-26, InRangeB=26, OutRangeA=-0.08, OutRangeB=0.08)
   n66 = float + float(A=float * float(A=@n64, B=-1), B=@n65)
   n67 = Get RightVelcoity() -> (RightVelcoity:double)
   n68 = "Sign (Float)"!(A:double=@n67) -> (ReturnValue:double)
   n69 = MapRangeClamped(Value=Get PelvisHeight() -> (PelvisHeight:double), InRangeA=80, InRangeB=42, OutRangeB=0.05)
   n70 = Get PawnVelocity()
   n71 = Get bIsOnGround()
   n72 = MapRangeClamped(Value=float * float(A=Vector Length(A=@n70), B="To Float (Boolean)"(InBool=@n71)), InRangeA=550, InRangeB=777, OutRangeB=0.04)
   n73 = Get TailMesh()
   n74 = BreakRotator(InRot=BreakTransform(InTransform=GetDeltaTransformFromRefPose(self=@n73, BoneName=C_Tail_A_Base_Jnt, BaseName=C_Tail_A_1_Jnt)).Rotation)
   n75 = GetDeltaTransformFromRefPose(self=Get TailMesh(), BoneName=C_Tail_A_2_Jnt, BaseName=C_Tail_A_1_Jnt)
   n76 = float * float(A=BreakRotator(InRot=BreakTransform(InTransform=@n75).Rotation), B=-1)
   n77 = Get bIsTargeting()
   n78 = BreakRotator(InRot=Get Spine_03_Rotation())
   n79 = float * float(
    A=MapRangeClamped(Value=SelectFloat(A=@n76, B=@n74, bPickA=@n77), InRangeA=-15, InRangeB=15, OutRangeA=0.12, OutRangeB=-0.12)
    B=MapRangeClamped(Value="Absolute (Float)"(A=@n78), InRangeA=0, InRangeB=26, OutRangeA=1, OutRangeB=2)
   )
   n80 = float + float(A=0.66, B=MapRangeClamped(Value=Get ForwardVelocity(), InRangeA=110, InRangeB=400, OutRangeB=0.05))
   > Set!(
    execute
    dy_shirt_bk_l_01_X:double=float + float!(
     A=@n80
     B=float * float(A=@n59, B=-1)
     C:double=@n69
     D:double=@n72
     E:double=@n79
     F:double=float * float(A=@n66, B=@n68)
     G:double=Get FastRotation() -> (FastRotation:double)
    ) -> (ReturnValue)
    dy_shirt_bk_l_01_Y:double=1
    dy_shirt_bk_l_01_Z:double=1
   ) -> (then, Output_Get)
   > n81 = Set dy_shirt_bk_l_02(dy_shirt_bk_l_02="0.85,1,1")
   > n82 = Set dy_shirt_bk_l_03(dy_shirt_bk_l_03="1,1,1")
   n83 = Get VelocityChange() -> (VelocityChange:double)
   n84 = Get RightVelcoity()
   n85 = Power!(Base:double=MapRangeClamped(Value=float * float!(A=@n83, B=@n84, C:double) -> (ReturnValue), InRangeA=20, InRangeB=130), Exp:double=0.5) -> (ReturnValue:double) {is=title K2Node_CallFunction}
   n86 = float + float!(A=@n64, B=@n65) -> (ReturnValue)
   n87 = MapRangeClamped(Value=Get RightVelcoity(), InRangeA=-70, InRangeB=-350, OutRangeB=-0.06)
   > Set!(
    execute
    dyn_shirt_bk_r_01_X:double=float + float!(
     A=@n80
     B=@n59
     C:double=@n69
     D:double=@n72
     E:double=float * float!(A=@n79, B=-1) -> (ReturnValue)
     F:double=float * float(A=@n87)
     G:double=float * float(A=@n86, B=@n68)
     H:double=Get FastRotation()
     I:double=Lerp(A=0, B=-0.03, Alpha=@n85)
    ) -> (ReturnValue)
    dyn_shirt_bk_r_01_Y:double=1
    dyn_shirt_bk_r_01_Z:double=1
   ) -> (then, Output_Get)
   > Set dyn_shirt_bk_r_02(dyn_shirt_bk_r_02=@n81)
   > Set dyn_shirt_bk_r_03(dyn_shirt_bk_r_03=@n82)
   n88 = BreakRotator(InRot=Get Spine_03_Rotation())
   n89 = SelectFloat(A=1, B=0, bPickA="float < float"(A=@n67))
   n90 = float + float!(A=FInterpEaseInOut(A=0, B=-0.3, Alpha=@n63, Exponent=2), B=@n65) -> (ReturnValue)
   n91 = Get TailMesh()
   n92 = BreakRotator(InRot=BreakTransform(InTransform=GetDeltaTransformFromRefPose(self=@n91, BoneName=C_Tail_A_Base_Jnt, BaseName=C_Tail_A_1_Jnt)).Rotation)
   n93 = GetDeltaTransformFromRefPose(self=Get TailMesh(), BoneName=C_Tail_A_2_Jnt, BaseName=C_Tail_A_1_Jnt)
   n94 = float * float(A=BreakRotator(InRot=BreakTransform(InTransform=@n93).Rotation), B=-1)
   n95 = Get bIsTargeting()
   n96 = BreakRotator(InRot=Get Spine_03_Rotation())
   n97 = float * float(
    A=MapRangeClamped(Value=SelectFloat(A=@n94, B=@n92, bPickA=@n95), InRangeA=-10, InRangeB=10, OutRangeA=0.09, OutRangeB=-0.09)
    B=MapRangeClamped(Value="Absolute (Float)"(A=@n96), InRangeA=0, InRangeB=26, OutRangeA=1, OutRangeB=2)
   )
   > Set!(
    execute
    dyn_shirt_l_01_X:double=float + float!(
     A=MapRangeClamped(Value=@n88, InRangeA=-10, InRangeB=-24, OutRangeA=1, OutRangeB=0.7)
     B=@n97
     C:double=float * float(A=@n90, B=@n89)
     D:double=Get FastRotation()
    ) -> (ReturnValue)
    dyn_shirt_l_01_Y:double=1
    dyn_shirt_l_01_Z:double=1
   ) -> (then, Output_Get)
   n98 = SelectFloat(A=1, B=0, bPickA="float > float"(A=@n67) {is=title K2Node_PromotableOperator})
   > Set!(
    execute
    dyn_shirt_r_01_X:double=float + float!(
     A=MapRangeClamped(Value=@n88, InRangeA=10, InRangeB=24, OutRangeA=1, OutRangeB=0.7)
     B=float * float(A=@n97, B=-1)
     C=float * float(A=@n90, B=@n98)
     D=Get FastRotation()
     E:double=Lerp(A=0, B=-0.1, Alpha=@n85)
    ) -> (ReturnValue)
    dyn_shirt_r_01_Y:double=1
    dyn_shirt_r_01_Z:double=1
   ) -> (then, Output_Get)
   n99 = NOT Boolean(A=Get bIsJumping())
   n100 = Get BodyMesh()
   n101 = BreakRotator(InRot=BreakTransform(InTransform=GetDeltaTransformFromRefPose(self=@n100, BoneName=thigh_r, BaseName=pelvis)).Rotation)
   n102 = SelectFloat(
    A=MapRangeClamped(Value=@n101.Yaw, InRangeA=-26, InRangeB=-90)
    B=MapRangeClamped(Value=@n101.Yaw, InRangeA=-20, InRangeB=-85)
    bPickA=NOT Boolean(A=Get bIsPlayingEmote())
   )
   n103 = "Max (Float)"(A=@n102, B=Get thigh rotation_r() -> (thigh rotation_r:double))
   n104 = SelectFloat(A=1, B=0.5, bPickA=NOT Boolean(A=Get bIsDBNO()))
   n105 = Set thigh rotation_r(execute=@n2.then_1, thigh rotation_r:double=float * float(A=SelectFloat(A=@n102, B=@n103, bPickA=@n99), B=@n104)) -> (Output_Get:double)
   n106 = NOT Boolean(A=Get bIsJumping())
   n107 = Get BodyMesh()
   n108 = BreakRotator(InRot=BreakTransform(InTransform=GetDeltaTransformFromRefPose(self=@n107, BoneName=thigh_l, BaseName=pelvis)).Rotation)
   n109 = SelectFloat(
    A=MapRangeClamped(Value=@n108.Yaw, InRangeA=-26, InRangeB=-90)
    B=MapRangeClamped(Value=@n108.Yaw, InRangeA=-20, InRangeB=-85)
    bPickA=NOT Boolean(A=Get bIsPlayingEmote())
   )
   n110 = "Max (Float)"(A=@n109, B=Get thigh rotation_r())
   > n111 = Set thigh_rotation_l(thigh_rotation_l:double=float * float(A=SelectFloat(A=@n109, B=@n110, bPickA=@n106), B=@n104)) -> (Output_Get:double)
   n112 = float / float(A=float + float(A=@n105, B=@n111) {is=title K2Node_CommutativeAssociativeBinaryOperator}, B=2) {is=title K2Node_CallFunction}
   n113 = float * float(A=Get AlphaMeleeAttack() -> (AlphaMeleeAttack:double))
   > Set thigh l plus r(thigh l plus r:double="Clamp (Float)"(Value=float + float!(A=@n112, B=@n113) -> (ReturnValue) {is=title K2Node_PromotableOperator}, Min=0, Max=1)) -> (Output_Get:double)
   Branch(execute=@n2.then_4, Condition=NOT Boolean(A=Get bIsSkyDiving()))
   n114 = Get PelvisHeight()
   n115 = "To Float (Boolean)"(InBool=Get bIsPlayingMeleeAnim())
   n116 = Get Delta Time X()
   n117 = Get AlphaMeleeAttack()
   > Set AlphaMeleeAttack(AlphaMeleeAttack:double="Clamp (Float)"(
     Value=FInterpTo(Current=@n117, Target=@n115, DeltaTime=@n116, InterpSpeed=7)
     Min=MapRangeClamped(Value=@n114, InRangeA=80, InRangeB=42, OutRangeA=0.4, OutRangeB=0.1)
     Max=0.4
    )) -> (Output_Get:double)
   n118 = float * float(A=BreakRotator(InRot=Get Spine_03_Rotation()))
   n119 = MapRangeClamped(Value=@n118, InRangeA=-20, InRangeB=-34)
   n120 = BreakRotator(InRot=Get Spine_03_Rotation())
   > Set!(
    execute
    Dyn_Shirt_frmid_l_02_Rotation_Roll:float=MapRangeClamped(Value=@n120, InRangeA=6, InRangeB=14, OutRangeB=-30)
    Dyn_Shirt_frmid_l_02_Rotation_Pitch:float=Lerp(A=0, B=-30, Alpha=@n119)
    Dyn_Shirt_frmid_l_02_Rotation_Yaw:float=0
   ) -> (then, Output_Get:Rotator)
   n121 = MapRangeClamped(Value=@n118, InRangeA=-10, InRangeB=-20)
   > Set!(
    execute
    Dyn_Shirt_frmid_l_04_Rotation_Roll:float=30
    Dyn_Shirt_frmid_l_04_Rotation_Pitch:float=float + float(A=MapRangeClamped(Value=@n120, InRangeA=6, InRangeB=14, OutRangeA=-30, OutRangeB=-40), B=Lerp(A=0, B=30, Alpha=@n121))
    Dyn_Shirt_frmid_l_04_Rotation_Yaw:float=0
   ) -> (then, Output_Get)
   > Set!(execute, Dyn_Shirt_l_02_scale_X:double=Lerp(A=0.65, B=1, Alpha=@n121), Dyn_Shirt_l_02_scale_Y:double=1, Dyn_Shirt_l_02_scale_Z:double=1) -> (then, Output_Get:Vector)
   > Set!(
    execute
    Dyn_Shirt_frmid_l_03_Rotation_Roll:float=0
    Dyn_Shirt_frmid_l_03_Rotation_Pitch:float=2
    Dyn_Shirt_frmid_l_03_Rotation_Yaw:float=Lerp(A=0, B=21, Alpha=@n121)
   ) -> (then, Output_Get:Rotator)
   n122 = Lerp(A=0, B=-5, Alpha=@n121)
   > Set Dyn_Shirt_frmid_l_03_transition(Dyn_Shirt_frmid_l_03_transition:Vector=vector + vector(A="0,1.8,0", B=MakeVector(Y=0, Z=@n122))) -> (Output_Get:Vector)
   > Set!(execute, Dyn_shirt_fr_02_scale_X:double=Lerp(A=0.7, B=0.55, Alpha=@n119), Dyn_shirt_fr_02_scale_Y:double=1, Dyn_shirt_fr_02_scale_Z:double=1) -> (then, Output_Get:Vector)
   > Set!(
    execute
    Dyn_shrit_l_02_rotation_Roll:float=0
    Dyn_shrit_l_02_rotation_Pitch:float=Lerp(A=0, B=10, Alpha=@n119)
    Dyn_shrit_l_02_rotation_Yaw:float=0
   ) -> (then, Output_Get:Rotator)
   > Set!(
    execute
    Dyn_Shirt_fr_02_rotation_Roll:float=20
    Dyn_Shirt_fr_02_rotation_Pitch:float=Lerp(A=10, B=-25, Alpha=@n119)
    Dyn_Shirt_fr_02_rotation_Yaw:float=-10
   ) -> (then, Output_Get)
   n123 = Get Spine_03_Rotation()
   > Set Alpha_Jersey_shirt_r_clipping(Alpha_Jersey_shirt_r_clipping:double=MapRangeClamped(Value=BreakRotator(InRot=@n123), InRangeA=10, InRangeB=25)) -> (Output_Get:double)
   n124 = Get bIsCrouching()
   > Set MeleeAttackStanding(MeleeAttackStanding:bool=AND Boolean(A=NOT Boolean(A=@n124), B=Get bIsPlayingMeleeAnim())) -> (Output_Get:bool)
   n125 = BreakTransform(InTransform=GetDeltaTransformFromRefPose(self=Get BodyMesh(), BoneName=clavicle_l, BaseName=spine_05))
   Set Pectoral_shirt_l_Alpha(execute=@n2.then_5, Pectoral_shirt_l_Alpha:double=MapRangeClamped(Value=BreakRotator(InRot=@n125.Rotation).Yaw, InRangeA=0, InRangeB=15)) -> (Output_Get:double)
   Event OnBeginCrouching()
   > Set crouch alpha(crouch alpha:double=1) -> (Output_Get:double)
   n126 = Event OnCharacterCustomizationLoaded()
   > Set BodyMesh(BodyMesh=@n126.CustomBaseMesh)
   > n127 = Cast To PlayerPawn_Athena(Object=TryGetPawnOwner())
   > n128 = Set PawnOwner(PawnOwner:Player Pawn Athena Object Reference=@n127) -> (Output_Get:Player Pawn Athena Object Reference)
   > Set TailMesh(TailMesh:Skeletal Mesh Component Object Reference=GetSkeletalMeshForPartType(self=@n128, PartType=MiscOrTail)) -> (Output_Get:Skeletal Mesh Component Object Reference)
   n129 = CheckForBackbling()
   > Set isWearingBackbling(isWearingBackbling=@n129)
   > Set isWearingCape(isWearingCape=@n129.bIsBackblingLong)
   > Set characterHasTail(characterHasTail:bool=CheckForTail!(self:Custom Character Part Anim Instance Object Reference) -> (bHasTail:bool)) -> (Output_Get:bool)
   > Branch(Condition=Get characterHasTail())
   > n130 = Branch(Condition=Get isWearingCape() -> (isWearingCape:bool))
   n131 = Get TailMesh()
   > SetVisibility!(execute, self:Scene Component Object Reference=@n131, bNewVisibility:bool=false, bPropagateToChildren:bool=false) -> (then)
   > Set cape and tail combo(cape and tail combo:bool=false) -> (Output_Get:bool)
   > n132 = Set cape and tail(cape and tail:bool=false) -> (Output_Get:bool)
   > Set TailAlpha(TailAlpha:double="To Float (Boolean)"(InBool=@n132)) -> (Output_Get:double)
   SetVisibility(execute=@n130.else, self=@n131, bNewVisibility=true, bPropagateToChildren=false)
   > n133 = Set cape and tail(cape and tail=true)
   > n134 = Branch(Condition=Get bIsInFrontEnd())
   > n135 = Set Cape and Tail Combo Frontend(Cape and Tail Combo Frontend:bool=true) -> (Output_Get:bool)
   > Set TailAlpha(TailAlpha="To Float (Boolean)"(InBool=@n135))
   n136 = Set cape and tail combo(execute=@n134.else, cape and tail combo=true)
   > Set TailAlpha(TailAlpha="To Float (Boolean)"(InBool=@n136))
   > Set TailAlpha(TailAlpha="To Float (Boolean)"(InBool=@n133))
   Event OnBeginKneeSliding() {is=title K2Node_Event}
   > Set KneeSliding(KneeSliding:double=1) -> (Output_Get:double)
   Event OnEndKneeSliding() {is=title K2Node_Event}
   > Set KneeSliding(KneeSliding=0)
   Event OnEndCrouching()
   > Set crouch alpha(crouch alpha=0)
   GetOwnerSkeletalMeshComponent()
   GetOwnerSkeletalMeshComponent()
   GetOwnerSkeletalMeshComponent()
   NOT Boolean(A=Get characterHasTail())
   GetOwnerSkeletalMeshComponent()
   MapRangeClamped(Value=@n101.Yaw, InRangeA=-26, InRangeB=-90, OutRangeB=0.5)
   GetOwnerSkeletalMeshComponent()
   NOT Boolean(A=Get bIsCrouching())
   GetOwnerSkeletalMeshComponent()
   AND Boolean(A=NOT Boolean(A=Get bIsCrouching()), B=Get bIsOnGround())
   AND Boolean(A=NOT Boolean(A=Get bIsCrouching()), B=Get bIsOnGround())
   OR Boolean(A=Get bIsOnGround(), B=Get bIsInFrontEnd(), C=Get bIsSkyDiving())
   Get bIsCrouching()
   Get bIsPlayingMeleeAnim()
   Get bIsCrouching()
  ]}
]
variables [
 Alpha_Jersey_shirt_r_clipping {type=number, default=1}
 AlphaFE {type=number}
 AlphaMeleeAttack {type=number, default=0.4}
 BodyMesh {type=SkeletalMeshComponent}
 cape and tail {type=boolean}
 cape and tail combo {type=boolean}
 Cape and Tail Combo Frontend {type=boolean}
 characterHasTail {type=boolean}
 Component Linear Acc Scale {type=Vector}
 Component Linear Vel Scale {type=Vector}
 ConstraintShirt_to_Tail {type=number, default=0.9}
 crouch alpha {type=number}
 DEBUG_tail {type=number, default=0.49}
 Delta Time X {type=number}
 dy_shirt_bk_l_01 {type=Vector, default=<0.7 1 1>}
 dy_shirt_bk_l_02 {type=Vector, default=<0.7 1 1>}
 dy_shirt_bk_l_03 {type=Vector, default=<0.7 1 1>}
 dyn_shirt_bk_01 {type=Vector, default=<0.7 1 1>}
 dyn_shirt_bk_02 {type=Vector, default=<0.8 1 1>}
 dyn_shirt_bk_03 {type=Vector, default=<0.95 1 1>}
 dyn_shirt_bk_r_01 {type=Vector, default=<0.7 1 1>}
 dyn_shirt_bk_r_02 {type=Vector, default=<0.7 1 1>}
 dyn_shirt_bk_r_03 {type=Vector, default=<0.7 1 1>}
 Dyn_Shirt_fr_02_rotation {type=Rotator, default=<rot 10 -10 20>}
 Dyn_shirt_fr_02_scale {type=Vector, default=<0.7 1 1>}
 Dyn_Shirt_frmid_l_02_Rotation {type=Rotator}
 Dyn_Shirt_frmid_l_03_Rotation {type=Rotator}
 Dyn_Shirt_frmid_l_03_transition {type=Vector}
 Dyn_Shirt_frmid_l_04_Rotation {type=Rotator}
 Dyn_Shirt_frmid_l_04_Scale {type=Vector, default=<1.5 1 1>}
 dyn_shirt_l_01 {type=Vector, default=<0.7 1 1>}
 Dyn_Shirt_l_02_scale {type=Vector, default=<0.65 1 1>}
 Dyn_Shirt_l_03_scale {type=Vector, default=<0.75 1 1>}
 dyn_shirt_r_01 {type=Vector, default=<0.7 1 1>}
 Dyn_shrit_l_02_rotation {type=Rotator}
 FastRotation {type=number}
 ForwardVelocity {type=number}
 gravity_override {type=Vector}
 History_position {type=PositionHistory, default {range=0, positions []}}
 History_rotation {type=PositionHistory, default {range=0, positions []}}
 IsCrouching+moving {type=number}
 isMoving {type=number}
 isWearingBackbling {type=boolean}
 isWearingCape {type=boolean}
 Jumping_alpha {type=number}
 KneeSliding {type=number}
 MeleeAttackStanding {type=boolean}
 NewVar {type=SkeletalMeshComponent}
 OnZipline {type=boolean}
 PawnOwner {type=PlayerPawn_Athena_C}
 Pectoral_shirt_l_Alpha {type=number}
 PelvisHeight {type=number}
 PevisPitchRotation {type=number}
 RightVelcoity {type=number}
 Shirt_bk_corecction_30pfs {type=number}
 SimlationSpace {type=AnimPhysSimSpaceType}
 Spine_03_Rotation {type=Rotator}
 TailAlpha {type=number}
 TailAlphaTransform {type=number}
 TailMesh {type=SkeletalMeshComponent}
 TailSwitchAgain {type=integer}
 thigh l plus r {type=number}
 thigh rotation_r {type=number}
 thigh_rotation_l {type=number}
 VelocityChange {type=number}
 ZiplineAlpha {type=number}
]
