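# Physics asset for the M_MED_Axl player body (notail variant) that sets up dynamic shirt motion against the lower torso and legs.
# Eight bone chains around the shirt hang from MSTR spheres tied to spine_03, using mostly capsules with per-joint swing and twist limits.
# Free-angular cross constraints link neighboring chains at the 03 and 04 tiers so panels move together, while pelvis, spine and thigh capsules supply body shapes.
# The frmid_r_02 to MSTR joint appears twice with different limits.
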PhysicsAsset /BRCosmetics/Characters/Player/Male/Medium/Bodies/M_MED_Axl/Meshes/M_MED_Axl_Dyn_notail
bodies {
 dyn_shirt_bk_01 [{shapeType=Capsule, length=4, radius=1.6, shapeName=dyn_shirt_bk_01_capsule, center=<2.6602 0 0>, rotation=<rot 90 0 0>}]
 dyn_shirt_bk_02 [{shapeType=Capsule, length=15, radius=2.8, shapeName=dyn_shirt_bk_02_capsule, center=<4 0 0>, rotation=<rot 0 0 90>}]
 dyn_shirt_bk_03 [{shapeType=Capsule, length=8, radius=2.8, shapeName=dyn_shirt_bk_03_capsule, center=<1.2 0 0>, rotation=<rot 0 90 90>}]
 dyn_shirt_bk_04 [{shapeType=Capsule, length=12, radius=3.5, shapeName=dyn_shirt_bk_04_capsule, center=<2 0 0>, rotation=<rot 0 0 90>}]
 dyn_shirt_bk_MSTR [{shapeType=Sphere, radius=0.505, shapeName=dyn_shirt_bk_MSTR_sphere}]
 dyn_shirt_bk_l_01 [{shapeType=Capsule, length=8.0768, radius=2.0933, shapeName=dyn_shirt_bk_l_01_capsule, center=<2.6602 0.0568 0.93>, rotation=<rot 0 0 90>}]
 dyn_shirt_bk_l_02 [{shapeType=Capsule, length=14, radius=2, shapeName=dyn_shirt_bk_l_02_capsule, center=<5 0 0>, rotation=<rot 0 0 90>}]
 dyn_shirt_bk_l_03 [{shapeType=Capsule, length=16, radius=2.8, shapeName=dyn_shirt_bk_l_03_capsule, rotation=<rot 0 0 90>}]
 dyn_shirt_bk_l_04 [{shapeType=Capsule, length=16, radius=2.8, shapeName=dyn_shirt_bk_l_04_capsule, rotation=<rot 0 0 90>}]
 dyn_shirt_bk_l_MSTR [{shapeType=Sphere, radius=0.505, shapeName=dyn_shirt_bk_l_MSTR_sphere}]
 dyn_shirt_bk_r_01 [{shapeType=Capsule, length=8.0768, radius=2.0933, shapeName=dyn_shirt_bk_r_01_capsule, center=<4 0.0568 0.93>, rotation=<rot 0 0 90>}]
 dyn_shirt_bk_r_02 [{shapeType=Capsule, length=14, radius=2, shapeName=dyn_shirt_bk_r_02_capsule, center=<5 0 0>, rotation=<rot 0 0 90>}]
 dyn_shirt_bk_r_03 [{shapeType=Capsule, length=16, radius=2.8, shapeName=dyn_shirt_bk_r_03_capsule, rotation=<rot 0 0 90>}]
 dyn_shirt_bk_r_04 [{shapeType=Capsule, length=16, radius=2.8, shapeName=dyn_shirt_bk_r_04_capsule, rotation=<rot 0 0 90>}]
 dyn_shirt_bk_r_MSTR [{shapeType=Sphere, radius=0.505, shapeName=dyn_shirt_bk_r_MSTR_sphere}]
 dyn_shirt_fr_01 [{shapeType=Capsule, length=3, radius=1.9133, shapeName=dyn_shirt_fr_01_capsule, center=<3 0.0568 0.0694>, rotation=<rot 0 90 90>}]
 dyn_shirt_fr_02 [{shapeType=Capsule, length=12, radius=2.2, shapeName=dyn_shirt_fr_02_capsule, center=<5 0 0>, rotation=<rot 0 0 90>}]
 dyn_shirt_fr_03 [{shapeType=Capsule, length=12, radius=2.8, shapeName=dyn_shirt_fr_03_capsule, center=<5 0 0>, rotation=<rot 90 180 180>}]
 dyn_shirt_fr_04 [{shapeType=Sphere, radius=3, shapeName=dyn_shirt_fr_04_sphere, center=<7 0 0>}]
 dyn_shirt_fr_MSTR [{shapeType=Sphere, radius=0.505, shapeName=dyn_shirt_fr_MSTR_sphere}]
 dyn_shirt_frmid_l_02 [{shapeType=Capsule, length=12, radius=2.2, shapeName=dyn_shirt_frmid_l_02_capsule, center=<4 0 0>, rotation=<rot 0 0 90>}]
 dyn_shirt_frmid_l_03 [{shapeType=Capsule, length=24, radius=2.2, shapeName=dyn_shirt_frmid_l_03_capsule, center=<11 0 0>, rotation=<rot 0 0 90>}]
 dyn_shirt_frmid_l_MSTR [{shapeType=Sphere, radius=0.505, shapeName=dyn_shirt_frmid_l_MSTR_sphere}]
 dyn_shirt_frmid_r_02 [{shapeType=Capsule, length=16, radius=2.2, shapeName=dyn_shirt_frmid_r_02_capsule, center=<4 0 0>, rotation=<rot 0 0 90>}]
 dyn_shirt_frmid_r_03 [{shapeType=Capsule, length=12, radius=2.4, shapeName=dyn_shirt_frmid_r_03_capsule, center=<5 0 0>, rotation=<rot 0 0 -90>}]
 dyn_shirt_frmid_r_04 [
  {shapeType=Capsule, length=26, radius=2.8, shapeName=dyn_shirt_frmid_r_04_capsule, center=<7.7 0 0>, rotation=<rot 0 0 90>}
  {
   shapeType=Capsule
   length=7
   radius=2.8
   shapeName=dyn_shirt_frmid_r_04_capsule
   center=<8.6145 -14.2482 -5.0783>
   rotation=<rot 10 3.7087 14.4735>
  }
 ]
 dyn_shirt_frmid_r_MSTR [{shapeType=Sphere, radius=0.505, shapeName=dyn_shirt_frmid_r_MSTR_sphere}]
 dyn_shirt_l_02 [{shapeType=Capsule, length=14, radius=2, shapeName=dyn_shirt_l_02_capsule, center=<2.1456 0 0.7009>, rotation=<rot 0 0 90>}]
 dyn_shirt_l_03 [{shapeType=Capsule, length=17, radius=2.4, shapeName=dyn_shirt_l_03_capsule, center=<7 0 2>, rotation=<rot 0 0 -90>}]
 dyn_shirt_l_04 [{shapeType=Capsule, length=22, radius=2.8, shapeName=dyn_shirt_l_04_capsule_mid, center=<8 0 2>, rotation=<rot 0 0 90>}]
 dyn_shirt_l_MSTR [{shapeType=Sphere, radius=0.505, shapeName=dyn_shirt_l_MSTR_sphere}]
 dyn_shirt_r_02 [{shapeType=Capsule, length=16, radius=2.4, shapeName=dyn_shirt_r_02_capsule, center=<4 0 1>, rotation=<rot 0 0 90>}]
 dyn_shirt_r_03 [
  {shapeType=Capsule, length=18, radius=2.8, shapeName=dyn_shirt_r_03_capsule, center=<0.5 0 1>, rotation=<rot 0 0 -90>}
  {shapeType=Capsule, length=12, radius=2.4, shapeName=dyn_shirt_r_03_capsule, center=<0.5 -9.1823 -5.8875>, rotation=<rot 0 0 10>}
 ]
 dyn_shirt_r_04 [{shapeType=Capsule, length=22, radius=2.6, shapeName=dyn_shirt_r_04_capsule, center=<0 0 1>, rotation=<rot 0 0 90>}]
 dyn_shirt_r_MSTR [{shapeType=Sphere, radius=0.505, shapeName=dyn_shirt_r_MSTR_sphere}]
 pelvis [{shapeType=Capsule, length=9.4, radius=9.4, center=<-4.5 -1 0>}]
 spine_01 [{shapeType=Capsule, length=10, radius=8, center=<0 -1 0>}]
 spine_02 [{shapeType=Capsule, length=9, radius=8.2, center=<0 -0.5 0>}]
 spine_03 [{shapeType=Capsule, length=7.6, radius=9.1, center=<3.1009 -0.2105 0>}]
 thigh_l [
  {shapeType=Capsule, length=40, radius=5, center=<-34 -4 0>, rotation=<rot 90 89.6476 79.6476>}
  {shapeType=Capsule, length=14, radius=8, shapeName=thigh_l_capsule, center=<-11 -1 0>, rotation=<rot 90 89.6476 84.6478>}
 ]
 thigh_r [
  {shapeType=Capsule, length=10, radius=8.1, center=<6 1 0>, rotation=<rot 90 -175 -178>}
  {shapeType=Capsule, length=40, radius=6, center=<34 3 0>, rotation=<rot 90 180 170>}
 ]
}
constraints [
 {bone1Name=dyn_shirt_bk_l_01, bone2Name=dyn_shirt_bk_l_MSTR, swing1LimitDegrees=15, swing2LimitDegrees=45, twistLimitDegrees=5}
 {bone1Name=dyn_shirt_bk_l_02, bone2Name=dyn_shirt_bk_l_01, swing1LimitDegrees=15, swing2LimitDegrees=45, twistLimitDegrees=5}
 {bone1Name=dyn_shirt_bk_l_03, bone2Name=dyn_shirt_bk_l_02, swing1LimitDegrees=15, swing2LimitDegrees=45, twistLimitDegrees=5}
 {
  bone1Name=dyn_shirt_bk_l_04
  bone2Name=dyn_shirt_bk_l_03
  swing1LimitDegrees=15
  swing1Motion=Locked
  swing2LimitDegrees=30
  twistLimitDegrees=5
  twistMotion=Locked
 }
 {bone1Name=dyn_shirt_fr_01, bone2Name=dyn_shirt_fr_MSTR, swing1LimitDegrees=5, swing2LimitDegrees=10, twistLimitDegrees=10}
 {bone1Name=dyn_shirt_fr_02, bone2Name=dyn_shirt_fr_01, swing1LimitDegrees=10, swing2LimitDegrees=40, twistLimitDegrees=20}
 {bone1Name=dyn_shirt_fr_03, bone2Name=dyn_shirt_fr_02, swing1LimitDegrees=30, swing2LimitDegrees=40, twistLimitDegrees=30}
 {bone1Name=dyn_shirt_fr_04, bone2Name=dyn_shirt_fr_03, swing1LimitDegrees=20, swing2LimitDegrees=30, twistLimitDegrees=10}
 {bone1Name=dyn_shirt_frmid_l_02, bone2Name=dyn_shirt_frmid_l_MSTR, swing1LimitDegrees=5, swing2LimitDegrees=10, twistLimitDegrees=10}
 {
  bone1Name=dyn_shirt_frmid_l_03
  bone2Name=dyn_shirt_frmid_l_02
  swing1LimitDegrees=10
  swing2LimitDegrees=50
  twistLimitDegrees=10
  twistMotion=Locked
 }
 {bone1Name=dyn_shirt_l_02, bone2Name=dyn_shirt_l_MSTR, swing1LimitDegrees=15, swing2LimitDegrees=40, twistLimitDegrees=5}
 {bone1Name=dyn_shirt_l_03, bone2Name=dyn_shirt_l_02, swing1LimitDegrees=15, swing2LimitDegrees=45, twistLimitDegrees=5}
 {bone1Name=dyn_shirt_l_04, bone2Name=dyn_shirt_l_03, swing1LimitDegrees=15, swing2LimitDegrees=15, twistLimitDegrees=15}
 {bone1Name=dyn_shirt_bk_01, bone2Name=dyn_shirt_bk_MSTR, swing1LimitDegrees=5, swing2LimitDegrees=20, twistLimitDegrees=5}
 {bone1Name=dyn_shirt_bk_02, bone2Name=dyn_shirt_bk_01, swing1LimitDegrees=5, swing2LimitDegrees=40, twistLimitDegrees=5}
 {bone1Name=dyn_shirt_bk_03, bone2Name=dyn_shirt_bk_02, swing1LimitDegrees=15, swing2LimitDegrees=60, twistLimitDegrees=5}
 {bone1Name=dyn_shirt_bk_04, bone2Name=dyn_shirt_bk_03, swing1LimitDegrees=12, swing2LimitDegrees=40, twistLimitDegrees=10, twistMotion=Locked}
 {bone1Name=dyn_shirt_r_02, bone2Name=dyn_shirt_r_MSTR, swing1LimitDegrees=10, swing2LimitDegrees=10, twistLimitDegrees=5, twistMotion=Locked}
 {bone1Name=dyn_shirt_r_03, bone2Name=dyn_shirt_r_02, swing1LimitDegrees=10, swing2LimitDegrees=80, twistLimitDegrees=5, twistMotion=Locked}
 {bone1Name=dyn_shirt_r_04, bone2Name=dyn_shirt_r_03, swing1LimitDegrees=10, swing2LimitDegrees=10, twistLimitDegrees=10, twistMotion=Locked}
 {
  bone1Name=dyn_shirt_frmid_r_02
  bone2Name=dyn_shirt_frmid_r_MSTR
  swing1LimitDegrees=1
  swing1Motion=Locked
  swing2LimitDegrees=60
  twistLimitDegrees=5
  twistMotion=Locked
 }
 {bone1Name=dyn_shirt_frmid_r_03, bone2Name=dyn_shirt_frmid_r_02, swing1LimitDegrees=5, swing2LimitDegrees=80, twistLimitDegrees=5}
 {bone1Name=dyn_shirt_frmid_r_04, bone2Name=dyn_shirt_frmid_r_03, swing1LimitDegrees=5, swing2LimitDegrees=20, twistLimitDegrees=5}
 {bone1Name=dyn_shirt_bk_r_04, bone2Name=dyn_shirt_bk_r_03, swing1LimitDegrees=5, swing2LimitDegrees=35, twistLimitDegrees=5}
 {bone1Name=dyn_shirt_bk_r_01, bone2Name=dyn_shirt_bk_r_MSTR, swing1LimitDegrees=15, swing2LimitDegrees=45, twistLimitDegrees=5}
 {bone1Name=dyn_shirt_bk_r_02, bone2Name=dyn_shirt_bk_r_01, swing1LimitDegrees=15, swing2LimitDegrees=45, twistLimitDegrees=5}
 {bone1Name=dyn_shirt_bk_r_03, bone2Name=dyn_shirt_bk_r_02, swing1LimitDegrees=5, swing2LimitDegrees=45, twistLimitDegrees=5}
 {
  bone1Name=dyn_shirt_bk_l_MSTR
  bone2Name=spine_03
  swing1LimitDegrees=45
  swing1Motion=Free
  swing2LimitDegrees=45
  swing2Motion=Free
  twistLimitDegrees=45
  twistMotion=Free
 }
 {
  bone1Name=dyn_shirt_fr_MSTR
  bone2Name=spine_03
  swing1LimitDegrees=45
  swing1Motion=Free
  swing2LimitDegrees=45
  swing2Motion=Free
  twistLimitDegrees=45
  twistMotion=Free
 }
 {
  bone1Name=dyn_shirt_frmid_l_MSTR
  bone2Name=spine_03
  swing1LimitDegrees=45
  swing1Motion=Free
  swing2LimitDegrees=45
  swing2Motion=Free
  twistLimitDegrees=45
  twistMotion=Free
 }
 {
  bone1Name=dyn_shirt_l_MSTR
  bone2Name=spine_03
  swing1LimitDegrees=45
  swing1Motion=Free
  swing2LimitDegrees=45
  swing2Motion=Free
  twistLimitDegrees=45
  twistMotion=Free
 }
 {
  bone1Name=dyn_shirt_bk_MSTR
  bone2Name=spine_03
  swing1LimitDegrees=45
  swing1Motion=Free
  swing2LimitDegrees=45
  swing2Motion=Free
  twistLimitDegrees=45
  twistMotion=Free
 }
 {
  bone1Name=dyn_shirt_r_MSTR
  bone2Name=spine_03
  swing1LimitDegrees=45
  swing1Motion=Free
  swing2LimitDegrees=45
  swing2Motion=Free
  twistLimitDegrees=45
  twistMotion=Free
 }
 {
  bone1Name=dyn_shirt_frmid_r_MSTR
  bone2Name=spine_03
  swing1LimitDegrees=45
  swing1Motion=Free
  swing2LimitDegrees=45
  swing2Motion=Free
  twistLimitDegrees=45
  twistMotion=Free
 }
 {
  bone1Name=dyn_shirt_bk_r_MSTR
  bone2Name=spine_03
  swing1LimitDegrees=45
  swing1Motion=Free
  swing2LimitDegrees=45
  swing2Motion=Free
  twistLimitDegrees=45
  twistMotion=Free
 }
 {
  bone1Name=spine_01
  bone2Name=pelvis
  swing1LimitDegrees=45
  swing1Motion=Free
  swing2LimitDegrees=45
  swing2Motion=Free
  twistLimitDegrees=45
  twistMotion=Free
 }
 {
  bone1Name=spine_02
  bone2Name=spine_01
  swing1LimitDegrees=45
  swing1Motion=Free
  swing2LimitDegrees=45
  swing2Motion=Free
  twistLimitDegrees=45
  twistMotion=Free
 }
 {
  bone1Name=spine_03
  bone2Name=spine_02
  swing1LimitDegrees=45
  swing1Motion=Free
  swing2LimitDegrees=45
  swing2Motion=Free
  twistLimitDegrees=45
  twistMotion=Free
 }
 {
  bone1Name=thigh_r
  bone2Name=pelvis
  swing1LimitDegrees=45
  swing1Motion=Free
  swing2LimitDegrees=45
  swing2Motion=Free
  twistLimitDegrees=45
  twistMotion=Free
 }
 {
  bone1Name=thigh_l
  bone2Name=pelvis
  swing1LimitDegrees=45
  swing1Motion=Free
  swing2LimitDegrees=45
  swing2Motion=Free
  twistLimitDegrees=45
  twistMotion=Free
 }
 {
  bone1Name=dyn_shirt_frmid_r_03
  bone2Name=dyn_shirt_fr_03
  swing1LimitDegrees=15
  swing1Motion=Free
  swing2LimitDegrees=45
  swing2Motion=Free
  twistLimitDegrees=5
  twistMotion=Free
 }
 {
  bone1Name=dyn_shirt_frmid_l_03
  bone2Name=dyn_shirt_fr_03
  swing1LimitDegrees=15
  swing1Motion=Free
  swing2LimitDegrees=45
  swing2Motion=Free
  twistLimitDegrees=5
  twistMotion=Free
 }
 {
  bone1Name=dyn_shirt_l_03
  bone2Name=dyn_shirt_bk_l_03
  swing1LimitDegrees=15
  swing1Motion=Free
  swing2LimitDegrees=45
  swing2Motion=Free
  twistLimitDegrees=5
  twistMotion=Free
 }
 {
  bone1Name=dyn_shirt_bk_03
  bone2Name=dyn_shirt_bk_l_03
  swing1LimitDegrees=15
  swing1Motion=Free
  swing2LimitDegrees=45
  swing2Motion=Free
  twistLimitDegrees=5
  twistMotion=Free
 }
 {
  bone1Name=dyn_shirt_bk_03
  bone2Name=dyn_shirt_bk_r_03
  swing1LimitDegrees=15
  swing1Motion=Free
  swing2LimitDegrees=45
  swing2Motion=Free
  twistLimitDegrees=5
  twistMotion=Free
 }
 {
  bone1Name=dyn_shirt_r_03
  bone2Name=dyn_shirt_bk_r_03
  swing1LimitDegrees=15
  swing1Motion=Free
  swing2LimitDegrees=45
  swing2Motion=Free
  twistLimitDegrees=5
  twistMotion=Free
 }
 {
  bone1Name=dyn_shirt_bk_l_04
  bone2Name=dyn_shirt_bk_04
  swing1LimitDegrees=45
  swing1Motion=Free
  swing2LimitDegrees=45
  swing2Motion=Free
  twistLimitDegrees=45
  twistMotion=Free
 }
 {
  bone1Name=dyn_shirt_bk_r_04
  bone2Name=dyn_shirt_bk_04
  swing1LimitDegrees=45
  swing1Motion=Free
  swing2LimitDegrees=45
  swing2Motion=Free
  twistLimitDegrees=45
  twistMotion=Free
 }
 {
  bone1Name=dyn_shirt_l_04
  bone2Name=dyn_shirt_bk_l_04
  swing1LimitDegrees=45
  swing1Motion=Free
  swing2LimitDegrees=45
  swing2Motion=Free
  twistLimitDegrees=45
  twistMotion=Free
 }
 {
  bone1Name=dyn_shirt_r_04
  bone2Name=dyn_shirt_bk_r_04
  swing1LimitDegrees=45
  swing1Motion=Free
  swing2LimitDegrees=45
  swing2Motion=Free
  twistLimitDegrees=45
  twistMotion=Free
 }
 {
  bone1Name=dyn_shirt_r_04
  bone2Name=dyn_shirt_frmid_r_04
  swing1LimitDegrees=45
  swing1Motion=Free
  swing2LimitDegrees=45
  swing2Motion=Free
  twistLimitDegrees=45
  twistMotion=Free
 }
 {
  bone1Name=dyn_shirt_fr_03
  bone2Name=dyn_shirt_l_03
  swing1LimitDegrees=45
  swing1Motion=Free
  swing2LimitDegrees=45
  swing2Motion=Free
  twistLimitDegrees=45
  twistMotion=Free
 }
 {
  bone1Name=dyn_shirt_frmid_r_04
  bone2Name=dyn_shirt_fr_04
  swing1LimitDegrees=45
  swing1Motion=Free
  swing2LimitDegrees=45
  swing2Motion=Free
  twistLimitDegrees=45
  twistMotion=Free
 }
 {
  bone1Name=dyn_shirt_frmid_r_03
  bone2Name=dyn_shirt_r_03
  swing1LimitDegrees=15
  swing1Motion=Free
  swing2LimitDegrees=45
  swing2Motion=Free
  twistLimitDegrees=5
  twistMotion=Free
 }
 {bone1Name=dyn_shirt_frmid_r_02, bone2Name=dyn_shirt_frmid_r_MSTR, swing1LimitDegrees=20, swing2LimitDegrees=10, twistLimitDegrees=10}
]
