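# Physics asset for the Axl male medium body cosmetic, defining collision and joints for its dynamic ear bones.
# Six three-bone chains (top, middle and bottom on each side) use identical capsules, with a separate capsule on the head.
# Each chain's root constrains to the head and each following bone to its predecessor, with uniform swing limits and a tight twist limit.
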PhysicsAsset /BRCosmetics/Characters/Player/Male/Medium/Bodies/M_MED_Axl/Meshes/Parts/M_MED_Axl_Dyn
bodies {
 dyn_ear_bott_l_01 [{shapeType=Capsule, length=3.5126, radius=3.7677, center=<4.2 0 -0.0001>, rotation=<rot 90 0 0>}]
 dyn_ear_bott_l_02 [{shapeType=Capsule, length=3.5126, radius=3.7677, center=<4.2 0 -0.0001>, rotation=<rot 90 0 0>}]
 dyn_ear_bott_l_03 [{shapeType=Capsule, length=3.5126, radius=3.7677, center=<4.2 0 -0.0001>, rotation=<rot 90 0 0>}]
 dyn_ear_bott_r_01 [{shapeType=Capsule, length=3.5126, radius=3.7677, center=<4.2 0 -0.0001>, rotation=<rot 90 0 0>}]
 dyn_ear_bott_r_02 [{shapeType=Capsule, length=3.5126, radius=3.7677, center=<4.2 0 -0.0001>, rotation=<rot 90 0 0>}]
 dyn_ear_bott_r_03 [{shapeType=Capsule, length=3.5126, radius=3.7677, center=<4.2 0 -0.0001>, rotation=<rot 90 0 0>}]
 dyn_ear_mid_l_01 [{shapeType=Capsule, length=3.5126, radius=3.7677, center=<4.2 0 -0.0001>, rotation=<rot 90 0 0>}]
 dyn_ear_mid_l_02 [{shapeType=Capsule, length=3.5126, radius=3.7677, center=<4.2 0 -0.0001>, rotation=<rot 90 0 0>}]
 dyn_ear_mid_l_03 [{shapeType=Capsule, length=3.5126, radius=3.7677, center=<4.2 0 -0.0001>, rotation=<rot 90 0 0>}]
 dyn_ear_mid_r_01 [{shapeType=Capsule, length=3.5126, radius=3.7677, center=<4.2 0 -0.0001>, rotation=<rot 90 0 0>}]
 dyn_ear_mid_r_02 [{shapeType=Capsule, length=3.5126, radius=3.7677, center=<4.2 0 -0.0001>, rotation=<rot 90 0 0>}]
 dyn_ear_mid_r_03 [{shapeType=Capsule, length=3.5126, radius=3.7677, center=<4.2 0 -0.0001>, rotation=<rot 90 0 0>}]
 dyn_ear_top_l_01 [{shapeType=Capsule, length=3.5126, radius=3.7677, center=<4.2 0 -0.0001>, rotation=<rot 90 0 0>}]
 dyn_ear_top_l_02 [{shapeType=Capsule, length=3.5126, radius=3.7677, center=<4.2 0 -0.0001>, rotation=<rot 90 0 0>}]
 dyn_ear_top_l_03 [{shapeType=Capsule, length=3.5126, radius=3.7677, center=<4.2 0 -0.0001>, rotation=<rot 90 0 0>}]
 dyn_ear_top_r_01 [{shapeType=Capsule, length=3.5126, radius=3.7677, center=<4.2 0 -0.0001>, rotation=<rot 90 0 0>}]
 dyn_ear_top_r_02 [{shapeType=Capsule, length=3.5126, radius=3.7677, center=<4.2 0 -0.0001>, rotation=<rot 90 0 0>}]
 dyn_ear_top_r_03 [{shapeType=Capsule, length=3.5126, radius=3.7677, center=<4.2 0 -0.0001>, rotation=<rot 90 0 0>}]
 head [{shapeType=Capsule, length=5, radius=5, rotation=<rot 90 0 0>}]
}
constraints [
 {bone1Name=dyn_ear_top_l_01, bone2Name=head, swing1LimitDegrees=45, swing2LimitDegrees=45, twistLimitDegrees=5}
 {bone1Name=dyn_ear_top_l_02, bone2Name=dyn_ear_top_l_01, swing1LimitDegrees=45, swing2LimitDegrees=45, twistLimitDegrees=5}
 {bone1Name=dyn_ear_top_l_03, bone2Name=dyn_ear_top_l_02, swing1LimitDegrees=45, swing2LimitDegrees=45, twistLimitDegrees=5}
 {bone1Name=dyn_ear_mid_l_01, bone2Name=head, swing1LimitDegrees=45, swing2LimitDegrees=45, twistLimitDegrees=5}
 {bone1Name=dyn_ear_mid_l_02, bone2Name=dyn_ear_mid_l_01, swing1LimitDegrees=45, swing2LimitDegrees=45, twistLimitDegrees=5}
 {bone1Name=dyn_ear_mid_l_03, bone2Name=dyn_ear_mid_l_02, swing1LimitDegrees=45, swing2LimitDegrees=45, twistLimitDegrees=5}
 {bone1Name=dyn_ear_bott_l_01, bone2Name=head, swing1LimitDegrees=45, swing2LimitDegrees=45, twistLimitDegrees=5}
 {bone1Name=dyn_ear_bott_l_02, bone2Name=dyn_ear_bott_l_01, swing1LimitDegrees=45, swing2LimitDegrees=45, twistLimitDegrees=5}
 {bone1Name=dyn_ear_bott_l_03, bone2Name=dyn_ear_bott_l_02, swing1LimitDegrees=45, swing2LimitDegrees=45, twistLimitDegrees=5}
 {bone1Name=dyn_ear_top_r_01, bone2Name=head, swing1LimitDegrees=45, swing2LimitDegrees=45, twistLimitDegrees=5}
 {bone1Name=dyn_ear_top_r_02, bone2Name=dyn_ear_top_r_01, swing1LimitDegrees=45, swing2LimitDegrees=45, twistLimitDegrees=5}
 {bone1Name=dyn_ear_top_r_03, bone2Name=dyn_ear_top_r_02, swing1LimitDegrees=45, swing2LimitDegrees=45, twistLimitDegrees=5}
 {bone1Name=dyn_ear_mid_r_01, bone2Name=head, swing1LimitDegrees=45, swing2LimitDegrees=45, twistLimitDegrees=5}
 {bone1Name=dyn_ear_mid_r_02, bone2Name=dyn_ear_mid_r_01, swing1LimitDegrees=45, swing2LimitDegrees=45, twistLimitDegrees=5}
 {bone1Name=dyn_ear_mid_r_03, bone2Name=dyn_ear_mid_r_02, swing1LimitDegrees=45, swing2LimitDegrees=45, twistLimitDegrees=5}
 {bone1Name=dyn_ear_bott_r_01, bone2Name=head, swing1LimitDegrees=45, swing2LimitDegrees=45, twistLimitDegrees=5}
 {bone1Name=dyn_ear_bott_r_02, bone2Name=dyn_ear_bott_r_01, swing1LimitDegrees=45, swing2LimitDegrees=45, twistLimitDegrees=5}
 {bone1Name=dyn_ear_bott_r_03, bone2Name=dyn_ear_bott_r_02, swing1LimitDegrees=45, swing2LimitDegrees=45, twistLimitDegrees=5}
]
