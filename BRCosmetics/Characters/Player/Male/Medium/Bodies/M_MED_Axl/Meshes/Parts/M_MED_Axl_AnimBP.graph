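# Post-process style AnimBP for the Axl medium male body part that adds secondary motion and corrective ear posing.
# It copies the attached parent mesh's pose, then runs RigidBody simulation relative to neck_01.
# Each update maps the spine_05-to-head rotation delta into squared, clamped left and right correction weights.
# These weights drive additive bottom, middle and top ear-bone rotations after physics, with interpolated alphas easing the release.
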AnimBlueprint /BRCosmetics/Characters/Player/Male/Medium/Bodies/M_MED_Axl/Meshes/Parts/M_MED_Axl_AnimBP
target_skeleton=&/M_MED_Axl_Skeleton
graphs [
 AnimGraph {nodes [
   n1 = RigidBody(ComponentPose=Local To Component(LocalPose=Copy Pose From Mesh() {bUseAttachedParent=true})) {
    bClampLinearTranslationLimitToRefPose=true
    bOverrideWorldGravity=true
    overridePhysicsAsset=&/M_MED_Axl_Dyn
    simulationSpace=BaseBoneSpace
    alphaBoolBlend {blendOption=Linear}
    baseBoneRef {boneName=neck_01}
    componentLinearAccScale=<0.4 0.3 0.3>
    componentLinearVelScale=<0.4 0.3 0.3>
    overrideWorldGravity {z=-980}
   }
   n2 = Get Dyn_ear_bot_r_correct() -> (Dyn_ear_bot_r_correct:double)
   n3 = "Transform (Modify) Bone - Bone: dyn_ear_bott_r_01"(ComponentPose=@n1, Alpha=@n2, Translation="0,0,0", Rotation="-60,0,-10", Scale="1,1,1") {
    rotationMode=BMM_Additive
    rotationSpace=BCS_BoneSpace
    alphaBoolBlend {blendOption=Linear}
    alphaScaleBiasClamp {bInterpResult=true, interpSpeedDecreasing=2}
    boneToModify {boneName=dyn_ear_bott_r_01}
    rotation {pitch=-60, roll=-10}
   }
   n4 = "Transform (Modify) Bone - Bone: dyn_ear_mid_r_01"(ComponentPose=@n3, Alpha=@n2, Translation="0,0,0", Rotation="-30,0,0", Scale="1,1,1") {
    rotationMode=BMM_Additive
    rotationSpace=BCS_BoneSpace
    alphaBoolBlend {blendOption=Linear}
    alphaScaleBiasClamp {bInterpResult=true, interpSpeedDecreasing=2}
    boneToModify {boneName=dyn_ear_mid_r_01}
    rotation {pitch=-30}
   }
   n5 = "Transform (Modify) Bone - Bone: dyn_ear_top_r_01"(ComponentPose=@n4, Alpha=@n2, Translation="0,0,0", Rotation="-15,0,0", Scale="1,1,1") {
    rotationMode=BMM_Additive
    rotationSpace=BCS_BoneSpace
    alphaBoolBlend {blendOption=Linear}
    alphaScaleBiasClamp {bInterpResult=true, interpSpeedDecreasing=2}
    boneToModify {boneName=dyn_ear_top_r_01}
    rotation {pitch=-15}
   }
   n6 = Get Dyn_ear_bot_l_correct() -> (Dyn_ear_bot_l_correct:double)
   n7 = "Transform (Modify) Bone - Bone: dyn_ear_bott_l_01"(ComponentPose=@n5, Alpha=@n6, Translation="0,0,0", Rotation="-60,-10,10", Scale="1,1,1") {
    rotationMode=BMM_Additive
    rotationSpace=BCS_BoneSpace
    alphaBoolBlend {blendOption=Linear}
    alphaScaleBiasClamp {bInterpResult=true, interpSpeedDecreasing=2}
    boneToModify {boneName=dyn_ear_bott_l_01}
    rotation=<rot -60 -10 10>
   }
   n8 = "Transform (Modify) Bone - Bone: dyn_ear_mid_l_01"(ComponentPose=@n7, Alpha=@n6, Translation="0,0,0", Rotation="-30,0,0", Scale="1,1,1") {
    rotationMode=BMM_Additive
    rotationSpace=BCS_BoneSpace
    alphaBoolBlend {blendOption=Linear}
    alphaScaleBiasClamp {bInterpResult=true, interpSpeedDecreasing=2}
    boneToModify {boneName=dyn_ear_mid_l_01}
    rotation {pitch=-30}
   }
   n9 = "Transform (Modify) Bone - Bone: dyn_ear_top_l_01"(ComponentPose=@n8, Alpha=@n6, Translation="0,0,0", Rotation="-15,0,0", Scale="1,1,1") {
    rotationMode=BMM_Additive
    rotationSpace=BCS_BoneSpace
    alphaBoolBlend {blendOption=Linear}
    alphaScaleBiasClamp {bInterpResult=true, interpSpeedDecreasing=2}
    boneToModify {boneName=dyn_ear_top_l_01}
    rotation {pitch=-15}
   }
   Output Pose(Result=Component To Local(ComponentPose=@n9)) {layerGroup=None}
  ]}
 EventGraph {nodes [
   Event BlueprintUpdateAnimation()
   n1 = GetOwnerSkeletalMeshComponent()
   > Set HeadRotation(HeadRotation:Rotator=BreakTransform(InTransform=GetDeltaTransformFromRefPose(self=@n1, BoneName=spine_05, BaseName=head)).Rotation) -> (Output_Get:Rotator)
   n2 = Get HeadRotation() -> (HeadRotation:Rotator)
   n3 = BreakRotator(InRot=@n2)
   n4 = float + float(
    A=float * float(A=MapRangeClamped(Value=@n3, InRangeA=-9, InRangeB=-16), B=MapRangeClamped(Value=@n3.Pitch, InRangeA=-12, InRangeB=-19)) {is=title K2Node_PromotableOperator}
    B=MapRangeClamped(Value=@n3.Yaw, InRangeA=8, InRangeB=40, OutRangeB=0.75)
   )
   > Set Dyn_ear_bot_r_correct(Dyn_ear_bot_r_correct:double=Power!(Base:double="Clamp (Float)"(Value=@n4, Min=0, Max=1), Exp:double=2) -> (ReturnValue:double) {is=title K2Node_CallFunction}) -> (Output_Get:double)
   n5 = BreakRotator(InRot=@n2)
   n6 = float + float(
    A=float * float(A=MapRangeClamped(Value=@n5, InRangeA=-9, InRangeB=-16), B=MapRangeClamped(Value=@n5.Pitch, InRangeA=12, InRangeB=19))
    B=MapRangeClamped(Value=@n5.Yaw, InRangeA=8, InRangeB=40, OutRangeB=0.75)
   )
   > Set Dyn_ear_bot_l_correct(Dyn_ear_bot_l_correct:double=Power(Base="Clamp (Float)"(Value=@n6, Min=0, Max=1), Exp=2)) -> (Output_Get:double)
  ]}
]
variables [
 Dyn_ear_bot_l_correct {type=number}
 Dyn_ear_bot_r_correct {type=number}
 HeadRotation {type=Rotator}
]
